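# Material function for the VerTet player body that draws an animated flipbook face (brows, eyes, mouth) onto an input material.
# Each feature calls MF_Flipbook with index parameters, its own texture, UVs, vertex-color channel and optional pose remapping, with left/right handling for brows and eyes.
# Results are blended into the material attributes, with distance-field line helpers, world-position-offset logic and debug parameters.
# Combined and per-feature masks, colors and alphas are also output.
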MaterialFunction /BRCosmetics/Characters/Player/Male/Medium/Bodies/M_MED_VerTet/Materials/MF_FlipbookFace_VerTet
graph {expressions {
  MaterialExpressionActorPositionWS_0 {class=MaterialExpressionActorPositionWS, inputs [], properties {}}
  MaterialExpressionAdd_0 {
   class=MaterialExpressionAdd
   inputs [
    A {from=MaterialExpressionStaticSwitch_17}
    B {from=MaterialExpressionStaticSwitch_18}
   ]
  }
  MaterialExpressionAdd_1 {
   class=MaterialExpressionAdd
   inputs [
    A {from=MaterialExpressionAdd_0}
    B {from=MaterialExpressionStaticSwitch_15}
   ]
  }
  MaterialExpressionAdd_10 {
   class=MaterialExpressionAdd
   inputs [
    A {from=MaterialExpressionMaterialFunctionCall_11, output=Result}
    B {from=MaterialExpressionVectorParameter_11, output=RGB}
   ]
  }
  MaterialExpressionAdd_2 {
   class=MaterialExpressionAdd
   inputs [
    A {from=MaterialExpressionNamedRerouteUsage_72}
    B {from=MaterialExpressionNamedRerouteUsage_73}
   ]
  }
  MaterialExpressionAdd_3 {
   class=MaterialExpressionAdd
   inputs [
    A {from=MaterialExpressionAdd_2}
    B {from=MaterialExpressionNamedRerouteUsage_74}
   ]
  }
  MaterialExpressionAdd_4 {
   class=MaterialExpressionAdd
   inputs [
    A {from=MaterialExpressionMaterialFunctionCall_13, output=Result}
    B {from=MaterialExpressionVectorParameter_10, output=RGB}
   ]
  }
  MaterialExpressionAdd_5 {
   class=MaterialExpressionAdd
   inputs [
    A {from=MaterialExpressionObjectPositionWS_1}
    B {from=MaterialExpressionAdd_10}
   ]
  }
  MaterialExpressionAdd_6 {
   class=MaterialExpressionAdd
   inputs [
    A {from=MaterialExpressionMaterialFunctionCall_0, output=R}
    B {from=MaterialExpressionScalarParameter_28}
   ]
  }
  MaterialExpressionAdd_7 {
   class=MaterialExpressionAdd
   inputs [
    A {from=MaterialExpressionMultiply_10}
    B {from=MaterialExpressionMultiply_7}
   ]
   properties {constA=100}
  }
  MaterialExpressionAppendVector_0 {
   class=MaterialExpressionAppendVector
   inputs [
    A {from=MaterialExpressionVectorParameter_1, output=""}
    B {from=MaterialExpressionVectorParameter_1, output=""}
   ]
  }
  MaterialExpressionAppendVector_1 {
   class=MaterialExpressionAppendVector
   inputs [
    A {from=MaterialExpressionVectorParameter_0, output=""}
    B {from=MaterialExpressionVectorParameter_0, output=""}
   ]
  }
  MaterialExpressionAppendVector_10 {
   class=MaterialExpressionAppendVector
   inputs [
    A {from=MaterialExpressionVectorParameter_9, output=""}
    B {from=MaterialExpressionVectorParameter_9, output=""}
   ]
  }
  MaterialExpressionAppendVector_11 {
   class=MaterialExpressionAppendVector
   inputs [
    A {from=MaterialExpressionScalarParameter_19, output=""}
    B {from=MaterialExpressionScalarParameter_20, output=""}
   ]
  }
  MaterialExpressionAppendVector_12 {
   class=MaterialExpressionAppendVector
   inputs [
    A {from=MaterialExpressionScalarParameter_18, output=""}
    B {from=MaterialExpressionScalarParameter_13, output=""}
   ]
  }
  MaterialExpressionAppendVector_13 {
   class=MaterialExpressionAppendVector
   inputs [
    A {from=MaterialExpressionScalarParameter_10, output=""}
    B {from=MaterialExpressionScalarParameter_15, output=""}
   ]
  }
  MaterialExpressionAppendVector_14 {
   class=MaterialExpressionAppendVector
   inputs [
    A {from=MaterialExpressionScalarParameter_23, output=""}
    B {from=MaterialExpressionScalarParameter_26, output=""}
   ]
  }
  MaterialExpressionAppendVector_2 {
   class=MaterialExpressionAppendVector
   inputs [
    A {from=MaterialExpressionVectorParameter_2, output=""}
    B {from=MaterialExpressionVectorParameter_2, output=""}
   ]
  }
  MaterialExpressionAppendVector_3 {
   class=MaterialExpressionAppendVector
   inputs [
    A {from=MaterialExpressionVectorParameter_4, output=""}
    B {from=MaterialExpressionVectorParameter_4, output=""}
   ]
  }
  MaterialExpressionAppendVector_4 {
   class=MaterialExpressionAppendVector
   inputs [
    A {from=MaterialExpressionVectorParameter_3, output=""}
    B {from=MaterialExpressionVectorParameter_3, output=""}
   ]
  }
  MaterialExpressionAppendVector_5 {
   class=MaterialExpressionAppendVector
   inputs [
    A {from=MaterialExpressionScalarParameter_11, output=""}
    B {from=MaterialExpressionScalarParameter_12, output=""}
   ]
  }
  MaterialExpressionAppendVector_6 {
   class=MaterialExpressionAppendVector
   inputs [
    A {from=MaterialExpressionScalarParameter_7, output=""}
    B {from=MaterialExpressionScalarParameter_8, output=""}
   ]
  }
  MaterialExpressionAppendVector_7 {
   class=MaterialExpressionAppendVector
   inputs [
    A {from=MaterialExpressionVectorParameter_7, output=""}
    B {from=MaterialExpressionVectorParameter_7, output=""}
   ]
  }
  MaterialExpressionAppendVector_8 {
   class=MaterialExpressionAppendVector
   inputs [
    A {from=MaterialExpressionVectorParameter_6, output=""}
    B {from=MaterialExpressionVectorParameter_6, output=""}
   ]
  }
  MaterialExpressionAppendVector_9 {
   class=MaterialExpressionAppendVector
   inputs [
    A {from=MaterialExpressionVectorParameter_8, output=""}
    B {from=MaterialExpressionVectorParameter_8, output=""}
   ]
  }
  MaterialExpressionChannelMaskParameter_0 {
   class=MaterialExpressionChannelMaskParameter
   inputs [None {from=MaterialExpressionNamedRerouteUsage_30, output=""}]
   properties {group=Flipbook 03 Eyes, maskChannel=Green, parameterName=FB_EyeVertexColorChannel, sortPriority=1, defaultValue=<rgba 0 1 0 0>}
  }
  MaterialExpressionChannelMaskParameter_1 {
   class=MaterialExpressionChannelMaskParameter
   inputs [None {from=MaterialExpressionNamedRerouteUsage_31, output=""}]
   properties {group=Flipbook 02 Brows, parameterName=FB_BrowVertexColorChannel, sortPriority=0}
  }
  MaterialExpressionChannelMaskParameter_2 {
   class=MaterialExpressionChannelMaskParameter
   inputs [None {from=MaterialExpressionNamedRerouteUsage_38, output=""}]
   properties {group=Flipbook 04 Mouth, maskChannel=Blue, parameterName=FB_MouthVertexColorChannel, sortPriority=0, defaultValue=<rgba 0 0 1 0>}
  }
  MaterialExpressionChannelMaskParameter_3 {
   class=MaterialExpressionChannelMaskParameter
   inputs [None {from=MaterialExpressionNamedRerouteUsage_79, output=""}]
   properties {group=Flipbook 02 Brows, maskChannel=Green, parameterName=FB_BrowMaskColorChannel, sortPriority=102, defaultValue=<rgba 0 1 0 0>}
  }
  MaterialExpressionChannelMaskParameter_4 {
   class=MaterialExpressionChannelMaskParameter
   inputs [None {from=MaterialExpressionNamedRerouteUsage_79, output=""}]
   properties {group=Flipbook 02 Brows, parameterName=FB_BrowMaskChannel, sortPriority=103}
  }
  MaterialExpressionChannelMaskParameter_5 {
   class=MaterialExpressionChannelMaskParameter
   inputs [None {from=MaterialExpressionNamedRerouteUsage_71, output=""}]
   properties {group=Flipbook 06 Debug, parameterName=DebugVertColorChannel, sortPriority=21}
  }
  MaterialExpressionComponentMask_1 {class=MaterialExpressionComponentMask, inputs [None {from=MaterialExpressionMultiply_6}], properties {r=true}}
  MaterialExpressionComponentMask_2 {class=MaterialExpressionComponentMask, inputs [None {from=MaterialExpressionMultiply_6}], properties {g=true}}
  MaterialExpressionConstant3Vector_0 {class=MaterialExpressionConstant3Vector, properties {constant=<rgba 1 1 1 0>}}
  MaterialExpressionConstant_0 {class=MaterialExpressionConstant}
  MaterialExpressionConstant_1 {class=MaterialExpressionConstant}
  MaterialExpressionConstant_10 {class=MaterialExpressionConstant}
  MaterialExpressionConstant_11 {class=MaterialExpressionConstant, properties {r=1}}
  MaterialExpressionConstant_12 {class=MaterialExpressionConstant}
  MaterialExpressionConstant_15 {class=MaterialExpressionConstant}
  MaterialExpressionConstant_18 {class=MaterialExpressionConstant}
  MaterialExpressionConstant_19 {class=MaterialExpressionConstant, properties {r=3}}
  MaterialExpressionConstant_2 {class=MaterialExpressionConstant}
  MaterialExpressionConstant_3 {class=MaterialExpressionConstant}
  MaterialExpressionConstant_4 {class=MaterialExpressionConstant}
  MaterialExpressionConstant_5 {class=MaterialExpressionConstant}
  MaterialExpressionConstant_6 {class=MaterialExpressionConstant}
  MaterialExpressionConstant_7 {class=MaterialExpressionConstant}
  MaterialExpressionConstant_8 {class=MaterialExpressionConstant}
  MaterialExpressionConstant_9 {class=MaterialExpressionConstant, properties {r=3}}
  MaterialExpressionConvert_0 {
   class=MaterialExpressionConvert
   inputs [
    None {from=MaterialExpressionGetMaterialAttributes_3, output=Specular}
    None {from=MaterialExpressionGetMaterialAttributes_3, output=Specular}
    None {from=MaterialExpressionGetMaterialAttributes_3, output=Specular}
   ]
   properties {
    nodeName=Make Vector3
    convertInputs [
     {
      type=Scalar
      expressionInput {
       expression=&self:MaterialExpressionGetMaterialAttributes_3
       inputName=None
       mask=0
       maskA=0
       maskB=0
       maskG=0
       maskR=0
       outputIndex=1
      }
     }
     {
      type=Scalar
      expressionInput {
       expression=&self:MaterialExpressionGetMaterialAttributes_3
       inputName=None
       mask=0
       maskA=0
       maskB=0
       maskG=0
       maskR=0
       outputIndex=2
      }
     }
     {
      type=Scalar
      expressionInput {
       expression=&self:MaterialExpressionGetMaterialAttributes_3
       inputName=None
       mask=0
       maskA=0
       maskB=0
       maskG=0
       maskR=0
       outputIndex=3
      }
     }
    ]
    convertOutputs [{type=Vector3}]
   }
  }
  MaterialExpressionConvert_1 {
   class=MaterialExpressionConvert
   inputs [None {from=MaterialExpressionLinearInterpolate_10}]
   properties {
    nodeName=Break Float3
    convertInputs [{
      type=Vector3
      expressionInput {
       expression=&self:MaterialExpressionLinearInterpolate_10
       inputName=None
       mask=0
       maskA=0
       maskB=0
       maskG=0
       maskR=0
       outputIndex=0
      }
     }]
    convertOutputs [{type=Scalar}, {type=Scalar}, {type=Scalar}]
   }
  }
  MaterialExpressionDivide_0 {
   class=MaterialExpressionDivide
   inputs [
    A {from=MaterialExpressionNamedRerouteUsage_98}
    B {from=MaterialExpressionNamedRerouteUsage_99}
   ]
   properties {constB=3}
  }
  MaterialExpressionDotProduct_0 {
   class=MaterialExpressionDotProduct
   inputs [
    A {from=MaterialExpressionConstant3Vector_0, output=""}
    B {from=MaterialExpressionVertexColor_3, output=""}
   ]
   properties {}
  }
  MaterialExpressionFloor_0 {class=MaterialExpressionFloor, inputs [None {from=MaterialExpressionDivide_0, output=""}], properties {}}
  MaterialExpressionFmod_0 {
   class=MaterialExpressionFmod
   inputs [
    A {from=MaterialExpressionNamedRerouteUsage_89, output=""}
    B {from=MaterialExpressionNamedRerouteUsage_113, output=""}
   ]
   properties {}
  }
  MaterialExpressionFunctionInput_0 {
   class=MaterialExpressionFunctionInput
   inputs [Preview {from=MaterialExpressionStaticSwitchParameter_2, output=""}]
   properties {bUsePreviewValueAsDefault=true, inputName=Mouth UV, inputType=FunctionInput_Vector2}
  }
  MaterialExpressionFunctionInput_1 {
   class=MaterialExpressionFunctionInput
   inputs [Preview {from=MaterialExpressionStaticSwitchParameter_0, output=""}]
   properties {bUsePreviewValueAsDefault=true, inputName=Eye UV, inputType=FunctionInput_Vector2}
  }
  MaterialExpressionFunctionInput_2 {
   class=MaterialExpressionFunctionInput
   inputs [Preview {from=MaterialExpressionStaticSwitchParameter_1, output=""}]
   properties {bUsePreviewValueAsDefault=true, inputName=Brow UV, inputType=FunctionInput_Vector2}
  }
  MaterialExpressionFunctionInput_3 {
   class=MaterialExpressionFunctionInput
   inputs [Preview {from=MaterialExpressionSetMaterialAttributes_1, output=""}]
   properties {bUsePreviewValueAsDefault=true, inputName=Material, inputType=FunctionInput_MaterialAttributes}
  }
  MaterialExpressionFunctionOutput_0 {
   class=MaterialExpressionFunctionOutput
   inputs [None {from=MaterialExpressionNamedRerouteDeclaration_58, output=""}]
   properties {outputName=BrowsColor, sortPriority=11}
  }
  MaterialExpressionFunctionOutput_1 {
   class=MaterialExpressionFunctionOutput
   inputs [None {from=MaterialExpressionNamedRerouteDeclaration_52, output=""}]
   properties {outputName=BrowsVertexMask, sortPriority=10}
  }
  MaterialExpressionFunctionOutput_10 {
   class=MaterialExpressionFunctionOutput
   inputs [None {from=MaterialExpressionNamedRerouteUsage_111, output=""}]
   properties {outputName=FaceColor, sortPriority=1}
  }
  MaterialExpressionFunctionOutput_11 {
   class=MaterialExpressionFunctionOutput
   inputs [None {from=MaterialExpressionNamedRerouteUsage_112, output=""}]
   properties {outputName=FaceAlpha, sortPriority=2}
  }
  MaterialExpressionFunctionOutput_12 {
   class=MaterialExpressionFunctionOutput
   inputs [None {from=MaterialExpressionNamedRerouteUsage_25, output=""}]
   properties {outputName=FaceVertexMask}
  }
  MaterialExpressionFunctionOutput_13 {class=MaterialExpressionFunctionOutput, inputs [None {from="", output=""}], properties {outputName=-----------------------, sortPriority=-1}}
  MaterialExpressionFunctionOutput_14 {class=MaterialExpressionFunctionOutput, inputs [None {from="", output=""}], properties {outputName=------------------------, sortPriority=9}}
  MaterialExpressionFunctionOutput_15 {
   class=MaterialExpressionFunctionOutput
   inputs [None {from="", output=""}]
   properties {outputName=-------------------------, sortPriority=19}
  }
  MaterialExpressionFunctionOutput_16 {
   class=MaterialExpressionFunctionOutput
   inputs [None {from="", output=""}]
   properties {outputName=--------------------------, sortPriority=29}
  }
  MaterialExpressionFunctionOutput_17 {
   class=MaterialExpressionFunctionOutput
   inputs [None {from=MaterialExpressionScalarParameter_5, output=""}]
   properties {outputName=flipbook_face_R_eye_index, sortPriority=33}
  }
  MaterialExpressionFunctionOutput_18 {
   class=MaterialExpressionFunctionOutput
   inputs [None {from=MaterialExpressionScalarParameter_4, output=""}]
   properties {outputName=flipbook_face_L_eye_index, sortPriority=34}
  }
  MaterialExpressionFunctionOutput_2 {
   class=MaterialExpressionFunctionOutput
   inputs [None {from=MaterialExpressionNamedRerouteDeclaration_59, output=""}]
   properties {outputName=BrowsAlpha, sortPriority=12}
  }
  MaterialExpressionFunctionOutput_3 {
   class=MaterialExpressionFunctionOutput
   inputs [None {from=MaterialExpressionNamedRerouteDeclaration_60, output=""}]
   properties {outputName=EyesVertexMask, sortPriority=20}
  }
  MaterialExpressionFunctionOutput_4 {
   class=MaterialExpressionFunctionOutput
   inputs [None {from=MaterialExpressionNamedRerouteDeclaration_61, output=""}]
   properties {outputName=EyesColor, sortPriority=21}
  }
  MaterialExpressionFunctionOutput_5 {
   class=MaterialExpressionFunctionOutput
   inputs [None {from=MaterialExpressionNamedRerouteDeclaration_62, output=""}]
   properties {outputName=EyesAlpha, sortPriority=22}
  }
  MaterialExpressionFunctionOutput_6 {
   class=MaterialExpressionFunctionOutput
   inputs [None {from=MaterialExpressionNamedRerouteDeclaration_63, output=""}]
   properties {outputName=MouthVertexMask, sortPriority=30}
  }
  MaterialExpressionFunctionOutput_7 {
   class=MaterialExpressionFunctionOutput
   inputs [None {from=MaterialExpressionNamedRerouteDeclaration_64, output=""}]
   properties {outputName=MouthColor, sortPriority=31}
  }
  MaterialExpressionFunctionOutput_8 {
   class=MaterialExpressionFunctionOutput
   inputs [None {from=MaterialExpressionNamedRerouteDeclaration_65, output=""}]
   properties {outputName=MouthAlpha, sortPriority=32}
  }
  MaterialExpressionFunctionOutput_9 {
   class=MaterialExpressionFunctionOutput
   inputs [None {from=MaterialExpressionNamedRerouteUsage_109, output=""}]
   properties {outputName=Material, sortPriority=-10}
  }
  MaterialExpressionGetMaterialAttributes_0 {
   class=MaterialExpressionGetMaterialAttributes
   inputs [None {from=MaterialExpressionNamedRerouteUsage_59, output=""}]
   properties {attributeGetTypes [#1]}
  }
  MaterialExpressionGetMaterialAttributes_1 {
   class=MaterialExpressionGetMaterialAttributes
   inputs [None {from=MaterialExpressionNamedRerouteUsage_88, output=""}]
   properties {attributeGetTypes [
     #1
     #2
     #3
     #4
    ]}
  }
  MaterialExpressionGetMaterialAttributes_3 {
   class=MaterialExpressionGetMaterialAttributes
   inputs [None {from=MaterialExpressionSetMaterialAttributes_0, output=""}]
   properties {attributeGetTypes [#5, #6, #7]}
  }
  MaterialExpressionIf_0 {
   class=MaterialExpressionIf
   inputs [
    A {from=MaterialExpressionDotProduct_0, output=""}
    B {from=MaterialExpressionConstant_9, output=""}
    "A > B" {from=MaterialExpressionConstant_10, output=""}
    "A == B" {from=MaterialExpressionConstant_10, output=""}
    "A < B" {from=MaterialExpressionConstant_11, output=""}
   ]
   properties {}
  }
  MaterialExpressionLinearInterpolate_0 {
   class=MaterialExpressionLinearInterpolate
   inputs [
    A {from=MaterialExpressionNamedRerouteUsage_114}
    B {from=MaterialExpressionNamedRerouteUsage_77}
    Alpha {from=MaterialExpressionNamedRerouteUsage_75}
   ]
  }
  MaterialExpressionLinearInterpolate_1 {
   class=MaterialExpressionLinearInterpolate
   inputs [
    A {from=MaterialExpressionStaticSwitch_13}
    B {from=MaterialExpressionNamedRerouteUsage_100}
    Alpha {from=MaterialExpressionNamedRerouteUsage_101}
   ]
  }
  MaterialExpressionLinearInterpolate_10 {
   class=MaterialExpressionLinearInterpolate
   inputs [
    A {from=MaterialExpressionConvert_0}
    B {from=MaterialExpressionVectorParameter_12, output=RGB}
    Alpha {from=MaterialExpressionNamedRerouteUsage_121}
   ]
  }
  MaterialExpressionLinearInterpolate_2 {
   class=MaterialExpressionLinearInterpolate
   inputs [
    A {from=MaterialExpressionReroute_3}
    B {from=MaterialExpressionNamedRerouteUsage_96}
    Alpha {from=MaterialExpressionNamedRerouteDeclaration_72}
   ]
  }
  MaterialExpressionLinearInterpolate_3 {
   class=MaterialExpressionLinearInterpolate
   inputs [
    A {from=MaterialExpressionGetMaterialAttributes_0, output=BaseColor}
    B {from=MaterialExpressionNamedRerouteUsage_70}
    Alpha {from=MaterialExpressionScalarParameter_27}
   ]
  }
  MaterialExpressionLinearInterpolate_4 {
   class=MaterialExpressionLinearInterpolate
   inputs [
    A {from=MaterialExpressionSubtract_2}
    B {from=MaterialExpressionNamedRerouteUsage_117}
    Alpha {from=MaterialExpressionNamedRerouteUsage_116}
   ]
  }
  MaterialExpressionLinearInterpolate_5 {
   class=MaterialExpressionLinearInterpolate
   inputs [
    A {from=MaterialExpressionGetMaterialAttributes_1, output=BaseColor}
    B {from=MaterialExpressionNamedRerouteUsage_107}
    Alpha {from=MaterialExpressionNamedRerouteUsage_106}
   ]
  }
  MaterialExpressionLinearInterpolate_6 {
   class=MaterialExpressionLinearInterpolate
   inputs [
    A {from=MaterialExpressionGetMaterialAttributes_1, output=EmissiveColor}
    B {from=MaterialExpressionMultiply_5}
    Alpha {from=MaterialExpressionNamedRerouteUsage_106}
   ]
  }
  MaterialExpressionLinearInterpolate_7 {
   class=MaterialExpressionLinearInterpolate
   inputs [
    A {from=MaterialExpressionIf_0}
    B {from=""}
    Alpha {from=MaterialExpressionScalarParameter_29}
   ]
   properties {constB=0}
  }
  MaterialExpressionLinearInterpolate_9 {
   class=MaterialExpressionLinearInterpolate
   inputs [
    A {from=MaterialExpressionGetMaterialAttributes_1, output=SubsurfaceColor}
    B {from=MaterialExpressionVectorParameter_13, output=RGB}
    Alpha {from=MaterialExpressionNamedRerouteUsage_106}
   ]
   properties {constB=0}
  }
  MaterialExpressionLocalPosition_0 {class=MaterialExpressionLocalPosition, inputs [], properties {localOrigin=InstancePreSkinning}}
  MaterialExpressionMaterialFunctionCall_0 {
   class=MaterialExpressionMaterialFunctionCall
   inputs [Float3 {from=MaterialExpressionShaderStageSwitch_0}]
   properties {materialFunction=&BreakOutFloat3Components}
  }
  MaterialExpressionMaterialFunctionCall_11 {
   class=MaterialExpressionMaterialFunctionCall
   inputs [
    X {from=MaterialExpressionSubtract_3}
    Y {from=MaterialExpressionConstant_15}
    Z {from=MaterialExpressionAdd_7}
   ]
   properties {materialFunction=&MakeFloat3}
  }
  MaterialExpressionMaterialFunctionCall_13 {
   class=MaterialExpressionMaterialFunctionCall
   inputs [
    X {from=MaterialExpressionSubtract_7}
    Y {from=MaterialExpressionConstant_18}
    Z {from=MaterialExpressionSubtract_8}
   ]
   properties {materialFunction=&MakeFloat3}
  }
  MaterialExpressionMaterialFunctionCall_2 {
   class=MaterialExpressionMaterialFunctionCall
   inputs [
    IndexRight {from=MaterialExpressionScalarParameter_5}
    IndexLeft {from=MaterialExpressionScalarParameter_4}
    LeftRightSync {from=MaterialExpressionStaticBoolParameter_9}
    UseLeftIndexForSync {from=MaterialExpressionStaticBoolParameter_0}
    "--------------------------      " {from=""}
    VertexColor {from=MaterialExpressionNamedRerouteUsage_27}
    ---------------------------- {from=""}
    Use Texture Flipbook {from=MaterialExpressionNamedRerouteUsage_1}
    Flipbook Texture {from=MaterialExpressionNamedRerouteUsage_2}
    Flipbook Column Count {from=MaterialExpressionNamedRerouteUsage_21}
    Flipbook Row Count {from=MaterialExpressionNamedRerouteUsage_7}
    UVs {from=MaterialExpressionNamedRerouteUsage_3}
    UV Scale {from=MaterialExpressionNamedRerouteUsage_22}
    UV Offset {from=MaterialExpressionNamedRerouteUsage_23}
    MirrorLeftUV {from=MaterialExpressionNamedRerouteUsage_24}
    ----------------------------- {from=""}
    Use Remap {from=MaterialExpressionNamedRerouteUsage_28}
    Remap Idle Only {from=MaterialExpressionNamedRerouteUsage_64}
    RemapPoseSync {from=MaterialExpressionNamedRerouteUsage_118}
    RemapIndex_0-3 {from=MaterialExpressionNamedRerouteUsage_13}
    RemapIndex_4-7 {from=MaterialExpressionNamedRerouteUsage_14}
    RemapIndex_8-11 {from=""}
    RemapIndex_12-15 {from=""}
    Remap Use 12 Inputs {from=""}
    Remap Use 16 Inputs {from=""}
   ]
   properties {materialFunction=/Game/Global/Materials/Character_Parent_2023/FlipbookFace_Parent_2023/MaterialFunctions/Common/MF_Flipbook}
  }
  MaterialExpressionMaterialFunctionCall_4 {
   class=MaterialExpressionMaterialFunctionCall
   inputs [
    IndexRight {from=MaterialExpressionScalarParameter_2}
    IndexLeft {from=MaterialExpressionScalarParameter_0}
    LeftRightSync {from=MaterialExpressionStaticBoolParameter_10}
    UseLeftIndexForSync {from=MaterialExpressionStaticBoolParameter_11}
    "--------------------------      " {from=""}
    VertexColor {from=MaterialExpressionNamedRerouteUsage_43}
    ---------------------------- {from=""}
    Use Texture Flipbook {from=MaterialExpressionNamedRerouteUsage_45}
    Flipbook Texture {from=MaterialExpressionNamedRerouteUsage_44}
    Flipbook Column Count {from=MaterialExpressionNamedRerouteUsage_46}
    Flipbook Row Count {from=MaterialExpressionNamedRerouteUsage_47}
    UVs {from=MaterialExpressionNamedRerouteUsage_48}
    UV Scale {from=MaterialExpressionNamedRerouteUsage_49}
    UV Offset {from=MaterialExpressionNamedRerouteUsage_50}
    MirrorLeftUV {from=MaterialExpressionNamedRerouteUsage_51}
    ----------------------------- {from=""}
    Use Remap {from=MaterialExpressionNamedRerouteUsage_42}
    Remap Idle Only {from=MaterialExpressionNamedRerouteUsage_76}
    RemapPoseSync {from=MaterialExpressionNamedRerouteUsage_115}
    RemapIndex_0-3 {from=MaterialExpressionNamedRerouteUsage_8}
    RemapIndex_4-7 {from=MaterialExpressionNamedRerouteUsage_9}
    RemapIndex_8-11 {from=MaterialExpressionNamedRerouteUsage_10}
    RemapIndex_12-15 {from=""}
    Remap Use 12 Inputs {from=MaterialExpressionStaticBool_1}
    Remap Use 16 Inputs {from=""}
   ]
   properties {materialFunction=&MF_Flipbook}
  }
  MaterialExpressionMaterialFunctionCall_6 {
   class=MaterialExpressionMaterialFunctionCall
   inputs [
    IndexRight {from=MaterialExpressionScalarParameter_1}
    IndexLeft {from=MaterialExpressionScalarParameter_1}
    LeftRightSync {from=MaterialExpressionStaticBool_0}
    UseLeftIndexForSync {from=""}
    "--------------------------      " {from=""}
    VertexColor {from=MaterialExpressionNamedRerouteUsage_41}
    ---------------------------- {from=""}
    Use Texture Flipbook {from=MaterialExpressionNamedRerouteUsage_33}
    Flipbook Texture {from=MaterialExpressionNamedRerouteUsage_34}
    Flipbook Column Count {from=MaterialExpressionNamedRerouteUsage_39}
    Flipbook Row Count {from=MaterialExpressionNamedRerouteUsage_40}
    UVs {from=MaterialExpressionNamedRerouteUsage_52}
    UV Scale {from=MaterialExpressionNamedRerouteUsage_36}
    UV Offset {from=MaterialExpressionNamedRerouteUsage_37}
    MirrorLeftUV {from=""}
    ----------------------------- {from=""}
    Use Remap {from=MaterialExpressionNamedRerouteUsage_35}
    Remap Idle Only {from=MaterialExpressionNamedRerouteUsage_32}
    RemapPoseSync {from=MaterialExpressionStaticBool_2}
    RemapIndex_0-3 {from=MaterialExpressionNamedRerouteUsage_17}
    RemapIndex_4-7 {from=MaterialExpressionNamedRerouteUsage_19}
    RemapIndex_8-11 {from=MaterialExpressionNamedRerouteUsage_20}
    RemapIndex_12-15 {from=MaterialExpressionNamedRerouteUsage_18}
    Remap Use 12 Inputs {from=MaterialExpressionStaticBool_3}
    Remap Use 16 Inputs {from=MaterialExpressionStaticBool_3}
   ]
   properties {materialFunction=&MF_Flipbook}
  }
  MaterialExpressionMaterialFunctionCall_7 {
   class=MaterialExpressionMaterialFunctionCall
   inputs [DF {from=MaterialExpressionNamedRerouteUsage_105}]
   properties {materialFunction=/Game/Global/Materials/Character_Parent_2023/FlipbookFace_Parent_2023/MaterialFunctions/Common/MF_Flipbook_DFLine}
  }
  MaterialExpressionMaterialFunctionCall_8 {
   class=MaterialExpressionMaterialFunctionCall
   inputs [DF {from=MaterialExpressionNamedRerouteUsage_104}]
   properties {materialFunction=&MF_Flipbook_DFLine}
  }
  MaterialExpressionMaterialFunctionCall_9 {
   class=MaterialExpressionMaterialFunctionCall
   inputs [DF {from=MaterialExpressionNamedRerouteUsage_103}]
   properties {materialFunction=&MF_Flipbook_DFLine}
  }
  MaterialExpressionMultiply_10 {
   class=MaterialExpressionMultiply
   inputs [
    A {from=MaterialExpressionScalarParameter_30}
    B {from=""}
   ]
   properties {constB=0.5}
  }
  MaterialExpressionMultiply_11 {
   class=MaterialExpressionMultiply
   inputs [
    A {from=MaterialExpressionStep_1}
    B {from=MaterialExpressionStep_2}
   ]
  }
  MaterialExpressionMultiply_12 {
   class=MaterialExpressionMultiply
   inputs [
    A {from=MaterialExpressionChannelMaskParameter_5, output=R}
    B {from=""}
   ]
   properties {constB=20}
  }
  MaterialExpressionMultiply_13 {
   class=MaterialExpressionMultiply
   inputs [
    A {from=MaterialExpressionFmod_0}
    B {from=MaterialExpressionNamedRerouteUsage_83}
   ]
   properties {constB=50}
  }
  MaterialExpressionMultiply_14 {
   class=MaterialExpressionMultiply
   inputs [
    A {from=MaterialExpressionNamedRerouteUsage_81}
    B {from=MaterialExpressionFloor_0}
   ]
  }
  MaterialExpressionMultiply_15 {
   class=MaterialExpressionMultiply
   inputs [
    A {from=MaterialExpressionMultiply_9}
    B {from=""}
   ]
   properties {constB=0.5}
  }
  MaterialExpressionMultiply_2 {
   class=MaterialExpressionMultiply
   inputs [
    A {from=MaterialExpressionVectorParameter_5, output=RGB}
    B {from=MaterialExpressionNamedRerouteDeclaration_71}
   ]
  }
  MaterialExpressionMultiply_3 {
   class=MaterialExpressionMultiply
   inputs [
    A {from=MaterialExpressionNamedRerouteUsage_58}
    B {from=MaterialExpressionNamedRerouteDeclaration_89}
   ]
  }
  MaterialExpressionMultiply_4 {
   class=MaterialExpressionMultiply
   inputs [
    A {from=MaterialExpressionNamedRerouteUsage_93}
    B {from=MaterialExpressionNamedRerouteUsage_87}
   ]
  }
  MaterialExpressionMultiply_5 {
   class=MaterialExpressionMultiply
   inputs [
    A {from=MaterialExpressionNamedRerouteUsage_107}
    B {from=MaterialExpressionScalarParameter_14}
   ]
  }
  MaterialExpressionMultiply_6 {
   class=MaterialExpressionMultiply
   inputs [
    A {from=MaterialExpressionStaticSwitchParameter_10}
    B {from=MaterialExpressionScalarParameter_30}
   ]
   properties {constB=100}
  }
  MaterialExpressionMultiply_7 {
   class=MaterialExpressionMultiply
   inputs [
    A {from=MaterialExpressionComponentMask_2}
    B {from=""}
   ]
   properties {constB=-1}
  }
  MaterialExpressionMultiply_9 {
   class=MaterialExpressionMultiply
   inputs [
    A {from=MaterialExpressionSubtract_6}
    B {from=MaterialExpressionNamedRerouteUsage_78}
   ]
  }
  MaterialExpressionNamedRerouteDeclaration_0 {
   class=MaterialExpressionNamedRerouteDeclaration
   inputs [None {from=MaterialExpressionStep_0}]
   properties {name=Right0/Left1, nodeColor=<rgba 1 0 0.1172 1>}
  }
  MaterialExpressionNamedRerouteDeclaration_1 {
   class=MaterialExpressionNamedRerouteDeclaration
   inputs [None {from=MaterialExpressionMaterialFunctionCall_2, output=Flipbook Vertex Mask}]
   properties {name=EyesVertexMaskBase, nodeColor=<rgba 0 1 0.7891 1>}
  }
  MaterialExpressionNamedRerouteDeclaration_10 {
   class=MaterialExpressionNamedRerouteDeclaration
   inputs [None {from=MaterialExpressionTextureObjectParameter_1}]
   properties {name=EyeFlipbookTexture, nodeColor=<rgba 1 0.1875 0 1>}
  }
  MaterialExpressionNamedRerouteDeclaration_11 {
   class=MaterialExpressionNamedRerouteDeclaration
   inputs [None {from=MaterialExpressionMaterialFunctionCall_6, output=FlipbookTexture}]
   properties {name=MouthColorBase, nodeColor=<rgba 0.8984 1 0 1>}
  }
  MaterialExpressionNamedRerouteDeclaration_12 {
   class=MaterialExpressionNamedRerouteDeclaration
   inputs [None {from=MaterialExpressionMaterialFunctionCall_6, output=FlipbookTextureAlpha}]
   properties {name=MouthAlphaBase, nodeColor=<rgba 0.8984 1 0 1>}
  }
  MaterialExpressionNamedRerouteDeclaration_13 {
   class=MaterialExpressionNamedRerouteDeclaration
   inputs [None {from=MaterialExpressionAppendVector_8}]
   properties {name=MouthRemap_0-3, nodeColor=<rgba 1 0 0.3984 1>}
  }
  MaterialExpressionNamedRerouteDeclaration_14 {
   class=MaterialExpressionNamedRerouteDeclaration
   inputs [None {from=MaterialExpressionAppendVector_7}]
   properties {name=MouthRemap_4-7, nodeColor=<rgba 1 0 0.3984 1>}
  }
  MaterialExpressionNamedRerouteDeclaration_15 {
   class=MaterialExpressionNamedRerouteDeclaration
   inputs [None {from=MaterialExpressionAppendVector_9}]
   properties {name=MouthRemap_8-11, nodeColor=<rgba 1 0 0.3984 1>}
  }
  MaterialExpressionNamedRerouteDeclaration_16 {
   class=MaterialExpressionNamedRerouteDeclaration
   inputs [None {from=MaterialExpressionAppendVector_10}]
   properties {name=MouthRemap_12-15, nodeColor=<rgba 1 0 0.3984 1>}
  }
  MaterialExpressionNamedRerouteDeclaration_17 {
   class=MaterialExpressionNamedRerouteDeclaration
   inputs [None {from=MaterialExpressionStaticBoolParameter_16}]
   properties {name=UseEyes, nodeColor=<rgba 0.1953 1 0 1>}
  }
  MaterialExpressionNamedRerouteDeclaration_18 {
   class=MaterialExpressionNamedRerouteDeclaration
   inputs [None {from=MaterialExpressionStaticBoolParameter_3}]
   properties {name=EyeUseTexture, nodeColor=<rgba 0.4531 1 0 1>}
  }
  MaterialExpressionNamedRerouteDeclaration_19 {
   class=MaterialExpressionNamedRerouteDeclaration
   inputs [None {from=MaterialExpressionScalarParameter_3}]
   properties {name=EyeRowCount, nodeColor=<rgba 1 0 0.7734 1>}
  }
  MaterialExpressionNamedRerouteDeclaration_2 {
   class=MaterialExpressionNamedRerouteDeclaration
   inputs [None {from=MaterialExpressionStaticBoolParameter_12}]
   properties {name=Use Distance Field Alpha, nodeColor=<rgba 0 1 0.3203 1>}
  }
  MaterialExpressionNamedRerouteDeclaration_20 {
   class=MaterialExpressionNamedRerouteDeclaration
   inputs [None {from=MaterialExpressionScalarParameter_9}]
   properties {name=EyeColumnCount, nodeColor=<rgba 1 0 0.7734 1>}
  }
  MaterialExpressionNamedRerouteDeclaration_21 {
   class=MaterialExpressionNamedRerouteDeclaration
   inputs [None {from=MaterialExpressionFunctionInput_1}]
   properties {name=EyeUVs, nodeColor=<rgba 1 0.5859 0 1>}
  }
  MaterialExpressionNamedRerouteDeclaration_22 {
   class=MaterialExpressionNamedRerouteDeclaration
   inputs [None {from=MaterialExpressionStaticSwitchParameter_4}]
   properties {name=EyeUVScale, nodeColor=<rgba 0.2656 0 1 1>}
  }
  MaterialExpressionNamedRerouteDeclaration_23 {
   class=MaterialExpressionNamedRerouteDeclaration
   inputs [None {from=MaterialExpressionAppendVector_5}]
   properties {name=EyeUVOffset, nodeColor=<rgba 0.0547 1 0 1>}
  }
  MaterialExpressionNamedRerouteDeclaration_24 {
   class=MaterialExpressionNamedRerouteDeclaration
   inputs [None {from=MaterialExpressionStaticBoolParameter_17}]
   properties {name=UseMouth, nodeColor=<rgba 0.1953 1 0 1>}
  }
  MaterialExpressionNamedRerouteDeclaration_25 {
   class=MaterialExpressionNamedRerouteDeclaration
   inputs [None {from=MaterialExpressionStaticBoolParameter_4}]
   properties {name=EyeMirrorLeftUV, nodeColor=<rgba 0 0.6953 1 1>}
  }
  MaterialExpressionNamedRerouteDeclaration_26 {
   class=MaterialExpressionNamedRerouteDeclaration
   inputs [None {from=MaterialExpressionSaturate_2}]
   properties {name=FaceVertexMask, nodeColor=<rgba 1 0.6328 0 1>}
  }
  MaterialExpressionNamedRerouteDeclaration_27 {
   class=MaterialExpressionNamedRerouteDeclaration
   inputs [None {from=MaterialExpressionSubtract_1}]
   properties {name=VertexColorHide, nodeColor=<rgba 0.6875 0 1 1>}
  }
  MaterialExpressionNamedRerouteDeclaration_28 {
   class=MaterialExpressionNamedRerouteDeclaration
   inputs [None {from=MaterialExpressionVertexColor_0}]
   properties {name=VertexColor, nodeColor=<rgba 0.6172 1 0 1>}
  }
  MaterialExpressionNamedRerouteDeclaration_29 {
   class=MaterialExpressionNamedRerouteDeclaration
   inputs [None {from=MaterialExpressionChannelMaskParameter_0, output=R}]
   properties {name=EyeVertexColor, nodeColor=<rgba 1 0 0.9375 1>}
  }
  MaterialExpressionNamedRerouteDeclaration_3 {
   class=MaterialExpressionNamedRerouteDeclaration
   inputs [None {from=MaterialExpressionAppendVector_1}]
   properties {name=BrowRemap_0-3, nodeColor=<rgba 1 0 0.3984 1>}
  }
  MaterialExpressionNamedRerouteDeclaration_30 {
   class=MaterialExpressionNamedRerouteDeclaration
   inputs [None {from=MaterialExpressionStaticBoolParameter_1}]
   properties {name=EyeUseRemap, nodeColor=<rgba 0 1 0.7422 1>}
  }
  MaterialExpressionNamedRerouteDeclaration_31 {
   class=MaterialExpressionNamedRerouteDeclaration
   inputs [None {from=MaterialExpressionTextureObjectParameter_2}]
   properties {name=BrowFlipbookTexture, nodeColor=<rgba 1 0.1875 0 1>}
  }
  MaterialExpressionNamedRerouteDeclaration_32 {
   class=MaterialExpressionNamedRerouteDeclaration
   inputs [None {from=MaterialExpressionStaticBoolParameter_6}]
   properties {name=BrowUseTexture, nodeColor=<rgba 0.4531 1 0 1>}
  }
  MaterialExpressionNamedRerouteDeclaration_33 {
   class=MaterialExpressionNamedRerouteDeclaration
   inputs [None {from=MaterialExpressionScalarParameter_16}]
   properties {name=BrowRowCount, nodeColor=<rgba 1 0 0.7734 1>}
  }
  MaterialExpressionNamedRerouteDeclaration_34 {
   class=MaterialExpressionNamedRerouteDeclaration
   inputs [None {from=MaterialExpressionScalarParameter_17}]
   properties {name=BrowColumnCount, nodeColor=<rgba 1 0 0.7734 1>}
  }
  MaterialExpressionNamedRerouteDeclaration_35 {
   class=MaterialExpressionNamedRerouteDeclaration
   inputs [None {from=MaterialExpressionFunctionInput_2}]
   properties {name=BrowUVs, nodeColor=<rgba 1 0.5859 0 1>}
  }
  MaterialExpressionNamedRerouteDeclaration_36 {
   class=MaterialExpressionNamedRerouteDeclaration
   inputs [None {from=MaterialExpressionStaticSwitchParameter_3}]
   properties {name=BrowUVScale, nodeColor=<rgba 0.2656 0 1 1>}
  }
  MaterialExpressionNamedRerouteDeclaration_37 {
   class=MaterialExpressionNamedRerouteDeclaration
   inputs [None {from=MaterialExpressionAppendVector_11}]
   properties {name=BrowUVOffset, nodeColor=<rgba 0.0547 1 0 1>}
  }
  MaterialExpressionNamedRerouteDeclaration_38 {
   class=MaterialExpressionNamedRerouteDeclaration
   inputs [None {from=MaterialExpressionStaticBoolParameter_7}]
   properties {name=BrowMirrorLeftUV, nodeColor=<rgba 0 0.6953 1 1>}
  }
  MaterialExpressionNamedRerouteDeclaration_39 {
   class=MaterialExpressionNamedRerouteDeclaration
   inputs [None {from=MaterialExpressionMultiply_3}]
   properties {name=Flipbook_WPO, nodeColor=<rgba 1 0 0.5391 1>}
  }
  MaterialExpressionNamedRerouteDeclaration_4 {
   class=MaterialExpressionNamedRerouteDeclaration
   inputs [None {from=MaterialExpressionAppendVector_0}]
   properties {name=BrowRemap_4-7, nodeColor=<rgba 1 0 0.3984 1>}
  }
  MaterialExpressionNamedRerouteDeclaration_40 {
   class=MaterialExpressionNamedRerouteDeclaration
   inputs [None {from=MaterialExpressionChannelMaskParameter_1, output=R}]
   properties {name=BrowVertexColor, nodeColor=<rgba 1 0 0.9375 1>}
  }
  MaterialExpressionNamedRerouteDeclaration_41 {
   class=MaterialExpressionNamedRerouteDeclaration
   inputs [None {from=MaterialExpressionStaticBoolParameter_5}]
   properties {name=BrowUseRemap, nodeColor=<rgba 0 1 0.7422 1>}
  }
  MaterialExpressionNamedRerouteDeclaration_42 {
   class=MaterialExpressionNamedRerouteDeclaration
   inputs [None {from=MaterialExpressionStaticBoolParameter_18}]
   properties {name=BrowRemapIdleOnly, nodeColor=<rgba 0 1 0.7422 1>}
  }
  MaterialExpressionNamedRerouteDeclaration_43 {
   class=MaterialExpressionNamedRerouteDeclaration
   inputs [None {from=MaterialExpressionStaticBoolParameter_8}]
   properties {name=UseMouthTexture, nodeColor=<rgba 1 0.7969 0 1>}
  }
  MaterialExpressionNamedRerouteDeclaration_44 {
   class=MaterialExpressionNamedRerouteDeclaration
   inputs [None {from=MaterialExpressionTextureObjectParameter_0}]
   properties {name=MouthFlipbookTexture, nodeColor=<rgba 0.4766 0 1 1>}
  }
  MaterialExpressionNamedRerouteDeclaration_45 {
   class=MaterialExpressionNamedRerouteDeclaration
   inputs [None {from=MaterialExpressionStaticSwitchParameter_9}]
   properties {name=MouthUVScale, nodeColor=<rgba 0 1 0.1562 1>}
  }
  MaterialExpressionNamedRerouteDeclaration_46 {
   class=MaterialExpressionNamedRerouteDeclaration
   inputs [None {from=MaterialExpressionAppendVector_6}]
   properties {name=MouthUVOffset, nodeColor=<rgba 1 0 0.1641 1>}
  }
  MaterialExpressionNamedRerouteDeclaration_47 {
   class=MaterialExpressionNamedRerouteDeclaration
   inputs [None {from=MaterialExpressionStaticBoolParameter_2}]
   properties {name=MouthUseRemap, nodeColor=<rgba 0 0.4844 1 1>}
  }
  MaterialExpressionNamedRerouteDeclaration_48 {
   class=MaterialExpressionNamedRerouteDeclaration
   inputs [None {from=MaterialExpressionChannelMaskParameter_2, output=R}]
   properties {name=MouthVertexColor, nodeColor=<rgba 1 0 0.9375 1>}
  }
  MaterialExpressionNamedRerouteDeclaration_49 {
   class=MaterialExpressionNamedRerouteDeclaration
   inputs [None {from=MaterialExpressionScalarParameter_21}]
   properties {name=MouthRowCount, nodeColor=<rgba 1 0 0.7734 1>}
  }
  MaterialExpressionNamedRerouteDeclaration_5 {
   class=MaterialExpressionNamedRerouteDeclaration
   inputs [None {from=MaterialExpressionStaticBoolParameter_15}]
   properties {name=UseBrows, nodeColor=<rgba 0.1953 1 0 1>}
  }
  MaterialExpressionNamedRerouteDeclaration_50 {
   class=MaterialExpressionNamedRerouteDeclaration
   inputs [None {from=MaterialExpressionScalarParameter_22}]
   properties {name=MouthColumnCount, nodeColor=<rgba 1 0 0.7734 1>}
  }
  MaterialExpressionNamedRerouteDeclaration_51 {
   class=MaterialExpressionNamedRerouteDeclaration
   inputs [None {from=MaterialExpressionFunctionInput_0}]
   properties {name=MouthUvs, nodeColor=<rgba 1 0.5859 0 1>}
  }
  MaterialExpressionNamedRerouteDeclaration_52 {
   class=MaterialExpressionNamedRerouteDeclaration
   inputs [None {from=MaterialExpressionStaticSwitch_0}]
   properties {name=BrowsVertexMask, nodeColor=<rgba 1 0 0.8437 1>}
  }
  MaterialExpressionNamedRerouteDeclaration_53 {
   class=MaterialExpressionNamedRerouteDeclaration
   inputs [None {from=MaterialExpressionMaterialFunctionCall_6, output=Flipbook Vertex Mask}]
   properties {name=MouthVertexMaskBase, nodeColor=<rgba 0.9453 1 0 1>}
  }
  MaterialExpressionNamedRerouteDeclaration_54 {
   class=MaterialExpressionNamedRerouteDeclaration
   inputs [None {from=MaterialExpressionMaterialFunctionCall_4, output=FlipbookTexture}]
   properties {name=BrowsColorBase, nodeColor=<rgba 1 0.4688 0 1>}
  }
  MaterialExpressionNamedRerouteDeclaration_55 {
   class=MaterialExpressionNamedRerouteDeclaration
   inputs [None {from=MaterialExpressionMaterialFunctionCall_4, output=FlipbookTextureAlpha}]
   properties {name=BrowsAlphaBase, nodeColor=<rgba 1 0.4688 0 1>}
  }
  MaterialExpressionNamedRerouteDeclaration_56 {
   class=MaterialExpressionNamedRerouteDeclaration
   inputs [None {from=MaterialExpressionMaterialFunctionCall_2, output=FlipbookTexture}]
   properties {name=EyesColorBase, nodeColor=<rgba 0 1 0.7891 1>}
  }
  MaterialExpressionNamedRerouteDeclaration_57 {
   class=MaterialExpressionNamedRerouteDeclaration
   inputs [None {from=MaterialExpressionMaterialFunctionCall_2, output=FlipbookTextureAlpha}]
   properties {name=EyesAlphaBase, nodeColor=<rgba 0 1 0.7891 1>}
  }
  MaterialExpressionNamedRerouteDeclaration_58 {
   class=MaterialExpressionNamedRerouteDeclaration
   inputs [None {from=MaterialExpressionStaticSwitch_1}]
   properties {name=BrowsColor, nodeColor=<rgba 1 0 0.8438 1>}
  }
  MaterialExpressionNamedRerouteDeclaration_59 {
   class=MaterialExpressionNamedRerouteDeclaration
   inputs [None {from=MaterialExpressionStaticSwitch_2}]
   properties {name=BrowsAlpha, nodeColor=<rgba 1 0 0.8438 1>}
  }
  MaterialExpressionNamedRerouteDeclaration_6 {
   class=MaterialExpressionNamedRerouteDeclaration
   inputs [None {from=MaterialExpressionAppendVector_2}]
   properties {name=BrowRemap_8-11, nodeColor=<rgba 1 0 0.3984 1>}
  }
  MaterialExpressionNamedRerouteDeclaration_60 {
   class=MaterialExpressionNamedRerouteDeclaration
   inputs [None {from=MaterialExpressionStaticSwitch_3}]
   properties {name=EyesVertexMask, nodeColor=<rgba 0.1953 0 1 1>}
  }
  MaterialExpressionNamedRerouteDeclaration_61 {
   class=MaterialExpressionNamedRerouteDeclaration
   inputs [None {from=MaterialExpressionStaticSwitch_4}]
   properties {name=EyesColor, nodeColor=<rgba 0.1953 0 1 1>}
  }
  MaterialExpressionNamedRerouteDeclaration_62 {
   class=MaterialExpressionNamedRerouteDeclaration
   inputs [None {from=MaterialExpressionStaticSwitch_5}]
   properties {name=EyesAlpha, nodeColor=<rgba 0.1953 0 1 1>}
  }
  MaterialExpressionNamedRerouteDeclaration_63 {
   class=MaterialExpressionNamedRerouteDeclaration
   inputs [None {from=MaterialExpressionStaticSwitch_6}]
   properties {name=MouthVertexMask, nodeColor=<rgba 0 0.7656 1 1>}
  }
  MaterialExpressionNamedRerouteDeclaration_64 {
   class=MaterialExpressionNamedRerouteDeclaration
   inputs [None {from=MaterialExpressionStaticSwitch_7}]
   properties {name=MouthColor, nodeColor=<rgba 0 0.7656 1 1>}
  }
  MaterialExpressionNamedRerouteDeclaration_65 {
   class=MaterialExpressionNamedRerouteDeclaration
   inputs [None {from=MaterialExpressionStaticSwitch_8}]
   properties {name=MouthAlpha, nodeColor=<rgba 0 0.7656 1 1>}
  }
  MaterialExpressionNamedRerouteDeclaration_66 {
   class=MaterialExpressionNamedRerouteDeclaration
   inputs [None {from=MaterialExpressionFunctionInput_3}]
   properties {name=IN_Material, nodeColor=<rgba 1 0.0938 0 1>}
  }
  MaterialExpressionNamedRerouteDeclaration_67 {
   class=MaterialExpressionNamedRerouteDeclaration
   inputs [None {from=MaterialExpressionOneMinus_0}]
   properties {name=BrowMask, nodeColor=<rgba 0 1 0.2969 1>}
  }
  MaterialExpressionNamedRerouteDeclaration_68 {
   class=MaterialExpressionNamedRerouteDeclaration
   inputs [None {from=MaterialExpressionStaticSwitch_9}]
   properties {name=BrowFinalColor, nodeColor=<rgba 1 0.1172 0 1>}
  }
  MaterialExpressionNamedRerouteDeclaration_69 {
   class=MaterialExpressionNamedRerouteDeclaration
   inputs [None {from=MaterialExpressionLinearInterpolate_4}]
   properties {name=DebugUVWPO, nodeColor=<rgba 0 0.4141 1 1>}
  }
  MaterialExpressionNamedRerouteDeclaration_7 {
   class=MaterialExpressionNamedRerouteDeclaration
   inputs [None {from=MaterialExpressionMaterialFunctionCall_4, output=Flipbook Vertex Mask}]
   properties {name=BrowsVertexMaskBase, nodeColor=<rgba 1 0.4688 0 1>}
  }
  MaterialExpressionNamedRerouteDeclaration_70 {
   class=MaterialExpressionNamedRerouteDeclaration
   inputs [None {from=MaterialExpressionStaticBoolParameter_13}]
   properties {name=BrowTextureAsMask, nodeColor=<rgba 1 0.2812 0 1>}
  }
  MaterialExpressionNamedRerouteDeclaration_71 {
   class=MaterialExpressionNamedRerouteDeclaration
   inputs [None {from=MaterialExpressionChannelMaskParameter_3, output=R}]
   properties {name=BrowMaskColorChannel, nodeColor=<rgba 0 0.0391 1 1>}
  }
  MaterialExpressionNamedRerouteDeclaration_72 {
   class=MaterialExpressionNamedRerouteDeclaration
   inputs [None {from=MaterialExpressionStaticSwitch_10}]
   properties {name=EyeMaskFinal, nodeColor=<rgba 1 0 0.5625 1>}
  }
  MaterialExpressionNamedRerouteDeclaration_73 {
   class=MaterialExpressionNamedRerouteDeclaration
   inputs [None {from=MaterialExpressionStaticSwitch_11}]
   properties {name=BrowMaskFinal, nodeColor=<rgba 0 0.8828 1 1>}
  }
  MaterialExpressionNamedRerouteDeclaration_74 {
   class=MaterialExpressionNamedRerouteDeclaration
   inputs [None {from=MaterialExpressionSaturate_1}]
   properties {name=FaceMask, nodeColor=<rgba 1 0.7969 0 1>}
  }
  MaterialExpressionNamedRerouteDeclaration_75 {
   class=MaterialExpressionNamedRerouteDeclaration
   inputs [None {from=MaterialExpressionStaticSwitch_14}]
   properties {name=FaceColor, nodeColor=<rgba 0.4766 0 1 1>}
  }
  MaterialExpressionNamedRerouteDeclaration_76 {
   class=MaterialExpressionNamedRerouteDeclaration
   inputs [None {from=MaterialExpressionStaticSwitchParameter_13}]
   properties {name=OUT_Material, nodeColor=<rgba 0 1 0.1562 1>}
  }
  MaterialExpressionNamedRerouteDeclaration_77 {
   class=MaterialExpressionNamedRerouteDeclaration
   inputs [None {from=MaterialExpressionStaticBoolParameter_19}]
   properties {name=EyeRemapIdleOnly, nodeColor=<rgba 0 1 0.7422 1>}
  }
  MaterialExpressionNamedRerouteDeclaration_78 {
   class=MaterialExpressionNamedRerouteDeclaration
   inputs [None {from=MaterialExpressionLinearInterpolate_0}]
   properties {name=DebugVertColorWPO, nodeColor=<rgba 1 0 0.8203 1>}
  }
  MaterialExpressionNamedRerouteDeclaration_79 {
   class=MaterialExpressionNamedRerouteDeclaration
   inputs [None {from=MaterialExpressionLinearInterpolate_3}]
   properties {name=DebugVertColorColor, nodeColor=<rgba 0.3828 0 1 1>}
  }
  MaterialExpressionNamedRerouteDeclaration_8 {
   class=MaterialExpressionNamedRerouteDeclaration
   inputs [None {from=MaterialExpressionAppendVector_4}]
   properties {name=EyeRemap_0-3, nodeColor=<rgba 1 0 0.3984 1>}
  }
  MaterialExpressionNamedRerouteDeclaration_80 {
   class=MaterialExpressionNamedRerouteDeclaration
   inputs [None {from=MaterialExpressionStaticBoolParameter_20}]
   properties {name=MouthRemapIdleOnly, nodeColor=<rgba 0 0.4844 1 1>}
  }
  MaterialExpressionNamedRerouteDeclaration_81 {
   class=MaterialExpressionNamedRerouteDeclaration
   inputs [None {from=MaterialExpressionMultiply_11}]
   properties {name=DebugVertexMask, nodeColor=<rgba 0.1953 1 0 1>}
  }
  MaterialExpressionNamedRerouteDeclaration_82 {
   class=MaterialExpressionNamedRerouteDeclaration
   inputs [None {from=MaterialExpressionSubtract_5}]
   properties {name=VertexCrunch, nodeColor=<rgba 1 0 0.5156 1>}
  }
  MaterialExpressionNamedRerouteDeclaration_83 {
   class=MaterialExpressionNamedRerouteDeclaration
   inputs [None {from=MaterialExpressionAdd_4}]
   properties {name=DebugVertexOffset, nodeColor=<rgba 0 0.8359 1 1>}
  }
  MaterialExpressionNamedRerouteDeclaration_84 {
   class=MaterialExpressionNamedRerouteDeclaration
   inputs [None {from=MaterialExpressionScalarParameter_33}]
   properties {name=DebugVertColorCellSize, nodeColor=<rgba 1 0.8438 0 1>}
  }
  MaterialExpressionNamedRerouteDeclaration_85 {
   class=MaterialExpressionNamedRerouteDeclaration
   inputs [None {from=MaterialExpressionConstant_19}]
   properties {name=DebugVertColorColumnCount, nodeColor=<rgba 1 0.2578 0 1>}
  }
  MaterialExpressionNamedRerouteDeclaration_86 {
   class=MaterialExpressionNamedRerouteDeclaration
   inputs [None {from=MaterialExpressionStaticBoolParameter_21}]
   properties {name=BrowUseRemapPoseSync, nodeColor=<rgba 0 1 0.7422 1>}
  }
  MaterialExpressionNamedRerouteDeclaration_87 {
   class=MaterialExpressionNamedRerouteDeclaration
   inputs [None {from=MaterialExpressionMultiply_15}]
   properties {name=DebugVertColorOffset, nodeColor=<rgba 0.7109 0 1 1>}
  }
  MaterialExpressionNamedRerouteDeclaration_88 {
   class=MaterialExpressionNamedRerouteDeclaration
   inputs [None {from=MaterialExpressionRound_1}]
   properties {name=DebugVertColorIndex, nodeColor=<rgba 0 1 0.3906 1>}
  }
  MaterialExpressionNamedRerouteDeclaration_89 {
   class=MaterialExpressionNamedRerouteDeclaration
   inputs [None {from=MaterialExpressionSaturate_0}]
   properties {name=VertexColorHideMask, nodeColor=<rgba 0 0.8359 1 1>}
  }
  MaterialExpressionNamedRerouteDeclaration_9 {
   class=MaterialExpressionNamedRerouteDeclaration
   inputs [None {from=MaterialExpressionAppendVector_3}]
   properties {name=EyeRemap_4-7, nodeColor=<rgba 1 0 0.3984 1>}
  }
  MaterialExpressionNamedRerouteDeclaration_90 {
   class=MaterialExpressionNamedRerouteDeclaration
   inputs [None {from=MaterialExpressionStaticBoolParameter_22}]
   properties {name=EyeUseRemapPoseSync, nodeColor=<rgba 0 1 0.7422 1>}
  }
  MaterialExpressionNamedRerouteUsage_0 {class=MaterialExpressionNamedRerouteUsage}
  MaterialExpressionNamedRerouteUsage_1 {class=MaterialExpressionNamedRerouteUsage}
  MaterialExpressionNamedRerouteUsage_10 {class=MaterialExpressionNamedRerouteUsage}
  MaterialExpressionNamedRerouteUsage_100 {class=MaterialExpressionNamedRerouteUsage}
  MaterialExpressionNamedRerouteUsage_101 {class=MaterialExpressionNamedRerouteUsage}
  MaterialExpressionNamedRerouteUsage_102 {class=MaterialExpressionNamedRerouteUsage}
  MaterialExpressionNamedRerouteUsage_103 {class=MaterialExpressionNamedRerouteUsage}
  MaterialExpressionNamedRerouteUsage_104 {class=MaterialExpressionNamedRerouteUsage}
  MaterialExpressionNamedRerouteUsage_105 {class=MaterialExpressionNamedRerouteUsage}
  MaterialExpressionNamedRerouteUsage_106 {class=MaterialExpressionNamedRerouteUsage}
  MaterialExpressionNamedRerouteUsage_107 {class=MaterialExpressionNamedRerouteUsage}
  MaterialExpressionNamedRerouteUsage_108 {class=MaterialExpressionNamedRerouteUsage}
  MaterialExpressionNamedRerouteUsage_109 {class=MaterialExpressionNamedRerouteUsage}
  MaterialExpressionNamedRerouteUsage_11 {class=MaterialExpressionNamedRerouteUsage}
  MaterialExpressionNamedRerouteUsage_110 {class=MaterialExpressionNamedRerouteUsage}
  MaterialExpressionNamedRerouteUsage_111 {class=MaterialExpressionNamedRerouteUsage}
  MaterialExpressionNamedRerouteUsage_112 {class=MaterialExpressionNamedRerouteUsage}
  MaterialExpressionNamedRerouteUsage_113 {class=MaterialExpressionNamedRerouteUsage}
  MaterialExpressionNamedRerouteUsage_114 {class=MaterialExpressionNamedRerouteUsage}
  MaterialExpressionNamedRerouteUsage_115 {class=MaterialExpressionNamedRerouteUsage}
  MaterialExpressionNamedRerouteUsage_116 {class=MaterialExpressionNamedRerouteUsage}
  MaterialExpressionNamedRerouteUsage_117 {class=MaterialExpressionNamedRerouteUsage}
  MaterialExpressionNamedRerouteUsage_118 {class=MaterialExpressionNamedRerouteUsage}
  MaterialExpressionNamedRerouteUsage_119 {class=MaterialExpressionNamedRerouteUsage}
  MaterialExpressionNamedRerouteUsage_12 {class=MaterialExpressionNamedRerouteUsage}
  MaterialExpressionNamedRerouteUsage_121 {class=MaterialExpressionNamedRerouteUsage}
  MaterialExpressionNamedRerouteUsage_13 {class=MaterialExpressionNamedRerouteUsage}
  MaterialExpressionNamedRerouteUsage_14 {class=MaterialExpressionNamedRerouteUsage}
  MaterialExpressionNamedRerouteUsage_15 {class=MaterialExpressionNamedRerouteUsage}
  MaterialExpressionNamedRerouteUsage_16 {class=MaterialExpressionNamedRerouteUsage}
  MaterialExpressionNamedRerouteUsage_17 {class=MaterialExpressionNamedRerouteUsage}
  MaterialExpressionNamedRerouteUsage_18 {class=MaterialExpressionNamedRerouteUsage}
  MaterialExpressionNamedRerouteUsage_19 {class=MaterialExpressionNamedRerouteUsage}
  MaterialExpressionNamedRerouteUsage_2 {class=MaterialExpressionNamedRerouteUsage}
  MaterialExpressionNamedRerouteUsage_20 {class=MaterialExpressionNamedRerouteUsage}
  MaterialExpressionNamedRerouteUsage_21 {class=MaterialExpressionNamedRerouteUsage}
  MaterialExpressionNamedRerouteUsage_22 {class=MaterialExpressionNamedRerouteUsage}
  MaterialExpressionNamedRerouteUsage_23 {class=MaterialExpressionNamedRerouteUsage}
  MaterialExpressionNamedRerouteUsage_24 {class=MaterialExpressionNamedRerouteUsage}
  MaterialExpressionNamedRerouteUsage_25 {class=MaterialExpressionNamedRerouteUsage}
  MaterialExpressionNamedRerouteUsage_26 {class=MaterialExpressionNamedRerouteUsage}
  MaterialExpressionNamedRerouteUsage_27 {class=MaterialExpressionNamedRerouteUsage}
  MaterialExpressionNamedRerouteUsage_28 {class=MaterialExpressionNamedRerouteUsage}
  MaterialExpressionNamedRerouteUsage_29 {class=MaterialExpressionNamedRerouteUsage}
  MaterialExpressionNamedRerouteUsage_3 {class=MaterialExpressionNamedRerouteUsage}
  MaterialExpressionNamedRerouteUsage_30 {class=MaterialExpressionNamedRerouteUsage}
  MaterialExpressionNamedRerouteUsage_31 {class=MaterialExpressionNamedRerouteUsage}
  MaterialExpressionNamedRerouteUsage_32 {class=MaterialExpressionNamedRerouteUsage}
  MaterialExpressionNamedRerouteUsage_33 {class=MaterialExpressionNamedRerouteUsage}
  MaterialExpressionNamedRerouteUsage_34 {class=MaterialExpressionNamedRerouteUsage}
  MaterialExpressionNamedRerouteUsage_35 {class=MaterialExpressionNamedRerouteUsage}
  MaterialExpressionNamedRerouteUsage_36 {class=MaterialExpressionNamedRerouteUsage}
  MaterialExpressionNamedRerouteUsage_37 {class=MaterialExpressionNamedRerouteUsage}
  MaterialExpressionNamedRerouteUsage_38 {class=MaterialExpressionNamedRerouteUsage}
  MaterialExpressionNamedRerouteUsage_39 {class=MaterialExpressionNamedRerouteUsage}
  MaterialExpressionNamedRerouteUsage_4 {class=MaterialExpressionNamedRerouteUsage}
  MaterialExpressionNamedRerouteUsage_40 {class=MaterialExpressionNamedRerouteUsage}
  MaterialExpressionNamedRerouteUsage_41 {class=MaterialExpressionNamedRerouteUsage}
  MaterialExpressionNamedRerouteUsage_42 {class=MaterialExpressionNamedRerouteUsage}
  MaterialExpressionNamedRerouteUsage_43 {class=MaterialExpressionNamedRerouteUsage}
  MaterialExpressionNamedRerouteUsage_44 {class=MaterialExpressionNamedRerouteUsage}
  MaterialExpressionNamedRerouteUsage_45 {class=MaterialExpressionNamedRerouteUsage}
  MaterialExpressionNamedRerouteUsage_46 {class=MaterialExpressionNamedRerouteUsage}
  MaterialExpressionNamedRerouteUsage_47 {class=MaterialExpressionNamedRerouteUsage}
  MaterialExpressionNamedRerouteUsage_48 {class=MaterialExpressionNamedRerouteUsage}
  MaterialExpressionNamedRerouteUsage_49 {class=MaterialExpressionNamedRerouteUsage}
  MaterialExpressionNamedRerouteUsage_5 {class=MaterialExpressionNamedRerouteUsage}
  MaterialExpressionNamedRerouteUsage_50 {class=MaterialExpressionNamedRerouteUsage}
  MaterialExpressionNamedRerouteUsage_51 {class=MaterialExpressionNamedRerouteUsage}
  MaterialExpressionNamedRerouteUsage_52 {class=MaterialExpressionNamedRerouteUsage}
  MaterialExpressionNamedRerouteUsage_53 {class=MaterialExpressionNamedRerouteUsage}
  MaterialExpressionNamedRerouteUsage_54 {class=MaterialExpressionNamedRerouteUsage}
  MaterialExpressionNamedRerouteUsage_55 {class=MaterialExpressionNamedRerouteUsage}
  MaterialExpressionNamedRerouteUsage_56 {class=MaterialExpressionNamedRerouteUsage}
  MaterialExpressionNamedRerouteUsage_57 {class=MaterialExpressionNamedRerouteUsage}
  MaterialExpressionNamedRerouteUsage_58 {class=MaterialExpressionNamedRerouteUsage}
  MaterialExpressionNamedRerouteUsage_59 {class=MaterialExpressionNamedRerouteUsage}
  MaterialExpressionNamedRerouteUsage_6 {class=MaterialExpressionNamedRerouteUsage}
  MaterialExpressionNamedRerouteUsage_60 {class=MaterialExpressionNamedRerouteUsage}
  MaterialExpressionNamedRerouteUsage_61 {class=MaterialExpressionNamedRerouteUsage}
  MaterialExpressionNamedRerouteUsage_62 {class=MaterialExpressionNamedRerouteUsage}
  MaterialExpressionNamedRerouteUsage_63 {class=MaterialExpressionNamedRerouteUsage}
  MaterialExpressionNamedRerouteUsage_64 {class=MaterialExpressionNamedRerouteUsage}
  MaterialExpressionNamedRerouteUsage_65 {class=MaterialExpressionNamedRerouteUsage}
  MaterialExpressionNamedRerouteUsage_66 {class=MaterialExpressionNamedRerouteUsage}
  MaterialExpressionNamedRerouteUsage_67 {class=MaterialExpressionNamedRerouteUsage}
  MaterialExpressionNamedRerouteUsage_68 {class=MaterialExpressionNamedRerouteUsage}
  MaterialExpressionNamedRerouteUsage_69 {class=MaterialExpressionNamedRerouteUsage}
  MaterialExpressionNamedRerouteUsage_7 {class=MaterialExpressionNamedRerouteUsage}
  MaterialExpressionNamedRerouteUsage_70 {class=MaterialExpressionNamedRerouteUsage}
  MaterialExpressionNamedRerouteUsage_71 {class=MaterialExpressionNamedRerouteUsage}
  MaterialExpressionNamedRerouteUsage_72 {class=MaterialExpressionNamedRerouteUsage}
  MaterialExpressionNamedRerouteUsage_73 {class=MaterialExpressionNamedRerouteUsage}
  MaterialExpressionNamedRerouteUsage_74 {class=MaterialExpressionNamedRerouteUsage}
  MaterialExpressionNamedRerouteUsage_75 {class=MaterialExpressionNamedRerouteUsage}
  MaterialExpressionNamedRerouteUsage_76 {class=MaterialExpressionNamedRerouteUsage}
  MaterialExpressionNamedRerouteUsage_77 {class=MaterialExpressionNamedRerouteUsage}
  MaterialExpressionNamedRerouteUsage_78 {class=MaterialExpressionNamedRerouteUsage}
  MaterialExpressionNamedRerouteUsage_79 {class=MaterialExpressionNamedRerouteUsage}
  MaterialExpressionNamedRerouteUsage_8 {class=MaterialExpressionNamedRerouteUsage}
  MaterialExpressionNamedRerouteUsage_80 {class=MaterialExpressionNamedRerouteUsage}
  MaterialExpressionNamedRerouteUsage_81 {class=MaterialExpressionNamedRerouteUsage}
  MaterialExpressionNamedRerouteUsage_82 {class=MaterialExpressionNamedRerouteUsage}
  MaterialExpressionNamedRerouteUsage_83 {class=MaterialExpressionNamedRerouteUsage}
  MaterialExpressionNamedRerouteUsage_84 {class=MaterialExpressionNamedRerouteUsage}
  MaterialExpressionNamedRerouteUsage_85 {class=MaterialExpressionNamedRerouteUsage}
  MaterialExpressionNamedRerouteUsage_86 {class=MaterialExpressionNamedRerouteUsage}
  MaterialExpressionNamedRerouteUsage_87 {class=MaterialExpressionNamedRerouteUsage}
  MaterialExpressionNamedRerouteUsage_88 {class=MaterialExpressionNamedRerouteUsage}
  MaterialExpressionNamedRerouteUsage_89 {class=MaterialExpressionNamedRerouteUsage}
  MaterialExpressionNamedRerouteUsage_9 {class=MaterialExpressionNamedRerouteUsage}
  MaterialExpressionNamedRerouteUsage_90 {class=MaterialExpressionNamedRerouteUsage}
  MaterialExpressionNamedRerouteUsage_91 {class=MaterialExpressionNamedRerouteUsage}
  MaterialExpressionNamedRerouteUsage_92 {class=MaterialExpressionNamedRerouteUsage}
  MaterialExpressionNamedRerouteUsage_93 {class=MaterialExpressionNamedRerouteUsage}
  MaterialExpressionNamedRerouteUsage_94 {class=MaterialExpressionNamedRerouteUsage}
  MaterialExpressionNamedRerouteUsage_95 {class=MaterialExpressionNamedRerouteUsage}
  MaterialExpressionNamedRerouteUsage_96 {class=MaterialExpressionNamedRerouteUsage}
  MaterialExpressionNamedRerouteUsage_97 {class=MaterialExpressionNamedRerouteUsage}
  MaterialExpressionNamedRerouteUsage_98 {class=MaterialExpressionNamedRerouteUsage}
  MaterialExpressionNamedRerouteUsage_99 {class=MaterialExpressionNamedRerouteUsage}
  MaterialExpressionObjectPositionWS_0 {class=MaterialExpressionObjectPositionWS, inputs [], properties {}}
  MaterialExpressionObjectPositionWS_1 {class=MaterialExpressionObjectPositionWS, inputs [], properties {}}
  MaterialExpressionOneMinus_0 {class=MaterialExpressionOneMinus, inputs [None {from=MaterialExpressionChannelMaskParameter_4, output=R}]}
  MaterialExpressionReroute_1 {class=MaterialExpressionReroute, inputs [None {from=MaterialExpressionReroute_3}]}
  MaterialExpressionReroute_3 {class=MaterialExpressionReroute, inputs [None {from=MaterialExpressionStaticSwitch_12}]}
  MaterialExpressionReroute_5 {class=MaterialExpressionReroute, inputs [None {from=MaterialExpressionReroute_6}]}
  MaterialExpressionReroute_6 {class=MaterialExpressionReroute, inputs [None {from=MaterialExpressionStaticSwitch_13}]}
  MaterialExpressionReroute_7 {class=MaterialExpressionReroute, inputs [None {from=MaterialExpressionReroute_8}]}
  MaterialExpressionReroute_8 {class=MaterialExpressionReroute, inputs [None {from=MaterialExpressionNamedRerouteUsage_88}]}
  MaterialExpressionRound_0 {class=MaterialExpressionRound, inputs [None {from=MaterialExpressionSubtract_11, output=""}], properties {}}
  MaterialExpressionRound_1 {class=MaterialExpressionRound, inputs [None {from=MaterialExpressionSubtract_4, output=""}], properties {}}
  MaterialExpressionSaturate_0 {class=MaterialExpressionSaturate, inputs [None {from=MaterialExpressionRound_0}]}
  MaterialExpressionSaturate_1 {class=MaterialExpressionSaturate, inputs [None {from=MaterialExpressionAdd_1}]}
  MaterialExpressionSaturate_2 {class=MaterialExpressionSaturate, inputs [None {from=MaterialExpressionAdd_3}]}
  MaterialExpressionScalarParameter_0 {
   class=MaterialExpressionScalarParameter
   properties {
    desc="Driven by Animation Blueprint.  Do not set manually."
    group=Flipbook 05 Indices
    parameterName=Flipbook_face_L_brow_index
    sortPriority=0
   }
  }
  MaterialExpressionScalarParameter_1 {
   class=MaterialExpressionScalarParameter
   properties {
    desc="Driven by Animation Blueprint.  Do not set manually."
    group=Flipbook 05 Indices
    parameterName=flipbook_face_mouth_index
    sortPriority=20
   }
  }
  MaterialExpressionScalarParameter_10 {class=MaterialExpressionScalarParameter, properties {defaultValue=1, group=Flipbook 03 Eyes, parameterName=FB_EyeUVScaleX, sortPriority=13}}
  MaterialExpressionScalarParameter_11 {class=MaterialExpressionScalarParameter, properties {group=Flipbook 03 Eyes, parameterName=FB_EyeUVOffsetX, sortPriority=16}}
  MaterialExpressionScalarParameter_12 {class=MaterialExpressionScalarParameter, properties {group=Flipbook 03 Eyes, parameterName=FB_EyeUVOffsetY, sortPriority=17}}
  MaterialExpressionScalarParameter_13 {
   class=MaterialExpressionScalarParameter
   properties {defaultValue=1, group=Flipbook 02 Brows, parameterName=FB_BrowUVScaleY, sortPriority=13}
  }
  MaterialExpressionScalarParameter_14 {class=MaterialExpressionScalarParameter, properties {defaultValue=1, group=Flipbook 01 Global, parameterName=FB_EmissiveStrength}}
  MaterialExpressionScalarParameter_15 {class=MaterialExpressionScalarParameter, properties {defaultValue=1, group=Flipbook 03 Eyes, parameterName=FB_EyeUVScaleY, sortPriority=14}}
  MaterialExpressionScalarParameter_16 {class=MaterialExpressionScalarParameter, properties {defaultValue=3, group=Flipbook 02 Brows, parameterName=FB_BrowRowCount, sortPriority=3}}
  MaterialExpressionScalarParameter_17 {
   class=MaterialExpressionScalarParameter
   properties {defaultValue=3, group=Flipbook 02 Brows, parameterName=FB_BrowColumnCount, sortPriority=4}
  }
  MaterialExpressionScalarParameter_18 {
   class=MaterialExpressionScalarParameter
   properties {defaultValue=1, group=Flipbook 02 Brows, parameterName=FB_BrowUVScaleX, sortPriority=12}
  }
  MaterialExpressionScalarParameter_19 {class=MaterialExpressionScalarParameter, properties {group=Flipbook 02 Brows, parameterName=FB_BrowUVOffsetX, sortPriority=20}}
  MaterialExpressionScalarParameter_2 {
   class=MaterialExpressionScalarParameter
   properties {
    desc="Driven by Animation Blueprint.  Do not set manually."
    group=Flipbook 05 Indices
    parameterName=Flipbook_face_R_brow_index
    sortPriority=1
   }
  }
  MaterialExpressionScalarParameter_20 {class=MaterialExpressionScalarParameter, properties {group=Flipbook 02 Brows, parameterName=FB_BrowUVOffsetY, sortPriority=21}}
  MaterialExpressionScalarParameter_21 {
   class=MaterialExpressionScalarParameter
   properties {defaultValue=4, group=Flipbook 04 Mouth, parameterName=FB_MouthRowCount, sortPriority=4}
  }
  MaterialExpressionScalarParameter_22 {
   class=MaterialExpressionScalarParameter
   properties {defaultValue=4, group=Flipbook 04 Mouth, parameterName=FB_MouthColumnCount, sortPriority=5}
  }
  MaterialExpressionScalarParameter_23 {
   class=MaterialExpressionScalarParameter
   properties {defaultValue=1, group=Flipbook 04 Mouth, parameterName=FB_MouthUVScaleX, sortPriority=13}
  }
  MaterialExpressionScalarParameter_24 {class=MaterialExpressionScalarParameter, properties {defaultValue=1, group=Flipbook 02 Brows, parameterName=FB_BrowUVScale, sortPriority=12}}
  MaterialExpressionScalarParameter_25 {class=MaterialExpressionScalarParameter, properties {defaultValue=1, group=Flipbook 03 Eyes, parameterName=FB_EyeUVScale, sortPriority=12}}
  MaterialExpressionScalarParameter_26 {
   class=MaterialExpressionScalarParameter
   properties {defaultValue=1, group=Flipbook 04 Mouth, parameterName=FB_MouthUVScaleY, sortPriority=14}
  }
  MaterialExpressionScalarParameter_27 {
   class=MaterialExpressionScalarParameter
   properties {defaultValue=1, group=Flipbook 06 Debug, parameterName=FB_DebugVertColor_Blend, sortPriority=24}
  }
  MaterialExpressionScalarParameter_28 {class=MaterialExpressionScalarParameter, properties {group=Flipbook 01 Global, parameterName=Right/Left_PositionOffset}}
  MaterialExpressionScalarParameter_29 {class=MaterialExpressionScalarParameter, properties {group=Flipbook 01 Global, parameterName=FB_DisableVertexColorLogic}}
  MaterialExpressionScalarParameter_3 {class=MaterialExpressionScalarParameter, properties {defaultValue=3, group=Flipbook 03 Eyes, parameterName=FB_EyeRowCount, sortPriority=4}}
  MaterialExpressionScalarParameter_30 {class=MaterialExpressionScalarParameter, properties {defaultValue=50, group=Flipbook 06 Debug, parameterName=FB_DebugUVSize, sortPriority=1}}
  MaterialExpressionScalarParameter_33 {
   class=MaterialExpressionScalarParameter
   properties {defaultValue=20, group=Flipbook 06 Debug, parameterName=FB_DebugVertColor_CellSize, sortPriority=22}
  }
  MaterialExpressionScalarParameter_4 {
   class=MaterialExpressionScalarParameter
   properties {
    desc="Driven by Animation Blueprint.  Do not set manually."
    group=Flipbook 05 Indices
    parameterName=flipbook_face_L_eye_index
    sortPriority=10
   }
  }
  MaterialExpressionScalarParameter_5 {
   class=MaterialExpressionScalarParameter
   properties {
    desc="Driven by Animation Blueprint.  Do not set manually."
    group=Flipbook 05 Indices
    parameterName=flipbook_face_R_eye_index
    sortPriority=11
   }
  }
  MaterialExpressionScalarParameter_6 {
   class=MaterialExpressionScalarParameter
   properties {defaultValue=1, group=Flipbook 04 Mouth, parameterName=FB_MouthUVScale, sortPriority=11}
  }
  MaterialExpressionScalarParameter_7 {class=MaterialExpressionScalarParameter, properties {group=Flipbook 04 Mouth, parameterName=FB_MouthUVOffsetX, sortPriority=16}}
  MaterialExpressionScalarParameter_8 {class=MaterialExpressionScalarParameter, properties {group=Flipbook 04 Mouth, parameterName=FB_MouthUVOffsetY, sortPriority=17}}
  MaterialExpressionScalarParameter_9 {
   class=MaterialExpressionScalarParameter
   properties {defaultValue=3, group=Flipbook 03 Eyes, parameterName=FB_EyeColumnCount, sortPriority=5}
  }
  MaterialExpressionSetMaterialAttributes_0 {
   class=MaterialExpressionSetMaterialAttributes
   inputs [
    MaterialAttributes {from=MaterialExpressionReroute_7, output=""}
    BaseColor {from=MaterialExpressionStaticSwitchParameter_7, output=""}
    EmissiveColor {from=MaterialExpressionStaticSwitchParameter_8, output=""}
    WorldPositionOffset {from=MaterialExpressionNamedRerouteUsage_108, output=""}
    OpacityMask {from=MaterialExpressionStaticSwitchParameter_5, output=""}
    SubsurfaceColor {from=MaterialExpressionStaticSwitchParameter_15, output=""}
   ]
   properties {attributeSetTypes [
     #1
     #2
     #8
     #9
     #4
    ]}
  }
  MaterialExpressionSetMaterialAttributes_1 {class=MaterialExpressionSetMaterialAttributes, inputs [MaterialAttributes {from="", output=""}]}
  MaterialExpressionSetMaterialAttributes_2 {
   class=MaterialExpressionSetMaterialAttributes
   inputs [
    MaterialAttributes {from=MaterialExpressionStaticSwitchParameter_12, output=""}
    BaseColor {from=MaterialExpressionNamedRerouteUsage_119, output=""}
   ]
   properties {attributeSetTypes [#1]}
  }
  MaterialExpressionSetMaterialAttributes_3 {
   class=MaterialExpressionSetMaterialAttributes
   inputs [
    MaterialAttributes {from=MaterialExpressionStaticSwitchParameter_6, output=""}
    WorldPositionOffset {from=MaterialExpressionNamedRerouteUsage_63, output=""}
   ]
   properties {attributeSetTypes [#8]}
  }
  MaterialExpressionSetMaterialAttributes_4 {
   class=MaterialExpressionSetMaterialAttributes
   inputs [
    MaterialAttributes {from=MaterialExpressionStaticSwitchParameter_11, output=""}
    WorldPositionOffset {from=MaterialExpressionNamedRerouteUsage_65, output=""}
    BaseColor {from=MaterialExpressionNamedRerouteUsage_69, output=""}
   ]
   properties {attributeSetTypes [#8, #1]}
  }
  MaterialExpressionSetMaterialAttributes_5 {
   class=MaterialExpressionSetMaterialAttributes
   inputs [
    MaterialAttributes {from=MaterialExpressionGetMaterialAttributes_3, output=MaterialAttributes}
    Specular {from=MaterialExpressionConvert_1, output=""}
    Roughness {from=MaterialExpressionConvert_1, output=""}
    Metallic {from=MaterialExpressionConvert_1, output=""}
   ]
   properties {attributeSetTypes [#5, #6, #7]}
  }
  MaterialExpressionShaderStageSwitch_0 {
   class=MaterialExpressionShaderStageSwitch
   inputs [
    PixelShader {from=MaterialExpressionVertexInterpolator_0, output=PS}
    VertexShader {from=MaterialExpressionLocalPosition_0, output=XYZ}
   ]
   properties {}
  }
  MaterialExpressionStaticBoolParameter_0 {
   class=MaterialExpressionStaticBoolParameter
   inputs []
   properties {group=Flipbook 03 Eyes, parameterName=FB_EyesUseLeftIndexForSync, sortPriority=101}
  }
  MaterialExpressionStaticBoolParameter_1 {class=MaterialExpressionStaticBoolParameter, inputs [], properties {group=Flipbook 03 Eyes, parameterName=FB_EyeUseRemap, sortPriority=190}}
  MaterialExpressionStaticBoolParameter_10 {
   class=MaterialExpressionStaticBoolParameter
   inputs []
   properties {group=Flipbook 02 Brows, parameterName=FB_BrowsSyncLeftRight, sortPriority=40}
  }
  MaterialExpressionStaticBoolParameter_11 {
   class=MaterialExpressionStaticBoolParameter
   inputs []
   properties {group=Flipbook 02 Brows, parameterName=FB_BrowsUseLeftIndexForSync, sortPriority=41}
  }
  MaterialExpressionStaticBoolParameter_12 {class=MaterialExpressionStaticBoolParameter, inputs [], properties {group=Flipbook 01 Global, parameterName=FB_UseDistanceFieldAlpha}}
  MaterialExpressionStaticBoolParameter_13 {
   class=MaterialExpressionStaticBoolParameter
   inputs []
   properties {group=Flipbook 02 Brows, parameterName=FB_BrowTextureAsMask, sortPriority=101}
  }
  MaterialExpressionStaticBoolParameter_14 {
   class=MaterialExpressionStaticBoolParameter
   inputs []
   properties {group=Flipbook 02 Brows, parameterName=FB_UseBrowColorChannelAsAlpha, sortPriority=110}
  }
  MaterialExpressionStaticBoolParameter_15 {
   class=MaterialExpressionStaticBoolParameter
   inputs []
   properties {defaultValue=true, group=Flipbook 01 Global, parameterName=FB_UseBrows, sortPriority=-10}
  }
  MaterialExpressionStaticBoolParameter_16 {
   class=MaterialExpressionStaticBoolParameter
   inputs []
   properties {defaultValue=true, group=Flipbook 01 Global, parameterName=FB_UseEyes, sortPriority=-9}
  }
  MaterialExpressionStaticBoolParameter_17 {
   class=MaterialExpressionStaticBoolParameter
   inputs []
   properties {defaultValue=true, group=Flipbook 01 Global, parameterName=FB_UseMouth, sortPriority=-8}
  }
  MaterialExpressionStaticBoolParameter_18 {
   class=MaterialExpressionStaticBoolParameter
   inputs []
   properties {group=Flipbook 02 Brows, parameterName=FB_BrowRemapIdleOnly, sortPriority=200}
  }
  MaterialExpressionStaticBoolParameter_19 {
   class=MaterialExpressionStaticBoolParameter
   inputs []
   properties {group=Flipbook 03 Eyes, parameterName=FB_EyeRemapIdleOnly, sortPriority=192}
  }
  MaterialExpressionStaticBoolParameter_2 {
   class=MaterialExpressionStaticBoolParameter
   inputs []
   properties {group=Flipbook 04 Mouth, parameterName=FB_UseMouthRemap, sortPriority=200}
  }
  MaterialExpressionStaticBoolParameter_20 {
   class=MaterialExpressionStaticBoolParameter
   inputs []
   properties {group=Flipbook 04 Mouth, parameterName=FB_MouthRemapIdleOnly, sortPriority=200}
  }
  MaterialExpressionStaticBoolParameter_21 {
   class=MaterialExpressionStaticBoolParameter
   inputs []
   properties {group=Flipbook 02 Brows, parameterName=FB_BrowRemapPoseSync, sortPriority=191}
  }
  MaterialExpressionStaticBoolParameter_22 {
   class=MaterialExpressionStaticBoolParameter
   inputs []
   properties {group=Flipbook 03 Eyes, parameterName=FB_EyeUseRemapPoseSync, sortPriority=191}
  }
  MaterialExpressionStaticBoolParameter_3 {class=MaterialExpressionStaticBoolParameter, inputs [], properties {group=Flipbook 03 Eyes, parameterName=FB_EyeUseTexture, sortPriority=2}}
  MaterialExpressionStaticBoolParameter_4 {
   class=MaterialExpressionStaticBoolParameter
   inputs []
   properties {group=Flipbook 03 Eyes, parameterName=FB_EyeMirrorLeftUV, sortPriority=18}
  }
  MaterialExpressionStaticBoolParameter_5 {
   class=MaterialExpressionStaticBoolParameter
   inputs []
   properties {group=Flipbook 02 Brows, parameterName=FB_BrowUseRemap, sortPriority=190}
  }
  MaterialExpressionStaticBoolParameter_6 {
   class=MaterialExpressionStaticBoolParameter
   inputs []
   properties {group=Flipbook 02 Brows, parameterName=FB_BrowUseTexture, sortPriority=1}
  }
  MaterialExpressionStaticBoolParameter_7 {
   class=MaterialExpressionStaticBoolParameter
   inputs []
   properties {group=Flipbook 02 Brows, parameterName=FB_BrowMirrorLeftUV, sortPriority=30}
  }
  MaterialExpressionStaticBoolParameter_8 {
   class=MaterialExpressionStaticBoolParameter
   inputs []
   properties {defaultValue=true, group=Flipbook 04 Mouth, parameterName=FB_MouthUseTexture, sortPriority=1}
  }
  MaterialExpressionStaticBoolParameter_9 {
   class=MaterialExpressionStaticBoolParameter
   inputs []
   properties {group=Flipbook 03 Eyes, parameterName=FB_EyesSyncLeftRight, sortPriority=100}
  }
  MaterialExpressionStaticBool_0 {class=MaterialExpressionStaticBool, properties {value=true}}
  MaterialExpressionStaticBool_1 {class=MaterialExpressionStaticBool, properties {value=true}}
  MaterialExpressionStaticBool_2 {class=MaterialExpressionStaticBool}
  MaterialExpressionStaticBool_3 {class=MaterialExpressionStaticBool, properties {value=true}}
  MaterialExpressionStaticSwitchParameter_0 {
   class=MaterialExpressionStaticSwitchParameter
   inputs [
    True {from=MaterialExpressionTextureCoordinate_1, output=""}
    False {from=MaterialExpressionTextureCoordinate_0, output=""}
   ]
   properties {group=Flipbook 03 Eyes, parameterName=FB_EyeUseUV2, sortPriority=10}
  }
  MaterialExpressionStaticSwitchParameter_1 {
   class=MaterialExpressionStaticSwitchParameter
   inputs [
    True {from=MaterialExpressionTextureCoordinate_3, output=""}
    False {from=MaterialExpressionTextureCoordinate_2, output=""}
   ]
   properties {group=Flipbook 02 Brows, parameterName=FB_BrowUseUV2, sortPriority=10}
  }
  MaterialExpressionStaticSwitchParameter_10 {
   class=MaterialExpressionStaticSwitchParameter
   inputs [
    True {from=MaterialExpressionTextureCoordinate_8, output=""}
    False {from=MaterialExpressionTextureCoordinate_7, output=""}
   ]
   properties {group=Flipbook 06 Debug, parameterName=FB_DebugUV2, sortPriority=0}
  }
  MaterialExpressionStaticSwitchParameter_11 {
   class=MaterialExpressionStaticSwitchParameter
   inputs [
    True {from=MaterialExpressionSetMaterialAttributes_3, output=""}
    False {from=MaterialExpressionStaticSwitchParameter_6, output=""}
   ]
   properties {group=Flipbook 06 Debug, parameterName=FB_UseDebugUVs, sortPriority=-1}
  }
  MaterialExpressionStaticSwitchParameter_12 {
   class=MaterialExpressionStaticSwitchParameter
   inputs [
    True {from=MaterialExpressionSetMaterialAttributes_4, output=""}
    False {from=MaterialExpressionStaticSwitchParameter_11, output=""}
   ]
   properties {group=Flipbook 06 Debug, parameterName=FB_UseDebugVertColor, sortPriority=20}
  }
  MaterialExpressionStaticSwitchParameter_13 {
   class=MaterialExpressionStaticSwitchParameter
   inputs [
    True {from=MaterialExpressionSetMaterialAttributes_2, output=""}
    False {from=MaterialExpressionStaticSwitchParameter_12, output=""}
   ]
   properties {group=Flipbook 06 Debug, parameterName=FB_UseDebugRight/Left_Position, sortPriority=20}
  }
  MaterialExpressionStaticSwitchParameter_15 {
   class=MaterialExpressionStaticSwitchParameter
   inputs [
    True {from=MaterialExpressionLinearInterpolate_9, output=""}
    False {from=MaterialExpressionGetMaterialAttributes_1, output=SubsurfaceColor}
   ]
   properties {group=Flipbook 01 Global, parameterName=FB_OverrideSubsurfaceColor, sortPriority=4}
  }
  MaterialExpressionStaticSwitchParameter_16 {
   class=MaterialExpressionStaticSwitchParameter
   inputs [
    True {from=MaterialExpressionSetMaterialAttributes_5, output=""}
    False {from=MaterialExpressionSetMaterialAttributes_0, output=""}
   ]
   properties {group=Flipbook 01 Global, parameterName=FB_OverrideSRM, sortPriority=4}
  }
  MaterialExpressionStaticSwitchParameter_2 {
   class=MaterialExpressionStaticSwitchParameter
   inputs [
    True {from=MaterialExpressionTextureCoordinate_5, output=""}
    False {from=MaterialExpressionTextureCoordinate_4, output=""}
   ]
   properties {group=Flipbook 04 Mouth, parameterName=FB_MouthUseUV2, sortPriority=10}
  }
  MaterialExpressionStaticSwitchParameter_3 {
   class=MaterialExpressionStaticSwitchParameter
   inputs [
    True {from=MaterialExpressionAppendVector_12, output=""}
    False {from=MaterialExpressionScalarParameter_24, output=""}
   ]
   properties {group=Flipbook 02 Brows, parameterName=FB_BrowUVNonUniformScale, sortPriority=11}
  }
  MaterialExpressionStaticSwitchParameter_4 {
   class=MaterialExpressionStaticSwitchParameter
   inputs [
    True {from=MaterialExpressionAppendVector_13, output=""}
    False {from=MaterialExpressionScalarParameter_25, output=""}
   ]
   properties {group=Flipbook 03 Eyes, parameterName=FB_EyeUVNonUniformScale, sortPriority=11}
  }
  MaterialExpressionStaticSwitchParameter_5 {
   class=MaterialExpressionStaticSwitchParameter
   inputs [
    True {from=MaterialExpressionNamedRerouteUsage_106, output=""}
    False {from=MaterialExpressionGetMaterialAttributes_1, output=Opacity}
   ]
   properties {group=Flipbook 01 Global, parameterName=FB_ApplyOpacityMask, sortPriority=4}
  }
  MaterialExpressionStaticSwitchParameter_6 {
   class=MaterialExpressionStaticSwitchParameter
   inputs [
    True {from=MaterialExpressionStaticSwitchParameter_16, output=""}
    False {from=MaterialExpressionNamedRerouteUsage_110, output=""}
   ]
   properties {defaultValue=true, group=Flipbook 01 Global, parameterName=FB_ApplyMaterialParams, sortPriority=0}
  }
  MaterialExpressionStaticSwitchParameter_7 {
   class=MaterialExpressionStaticSwitchParameter
   inputs [
    True {from=MaterialExpressionLinearInterpolate_5, output=""}
    False {from=MaterialExpressionGetMaterialAttributes_1, output=BaseColor}
   ]
   properties {defaultValue=true, group=Flipbook 01 Global, parameterName=FB_ApplyBaseColor, sortPriority=2}
  }
  MaterialExpressionStaticSwitchParameter_8 {
   class=MaterialExpressionStaticSwitchParameter
   inputs [
    True {from=MaterialExpressionLinearInterpolate_6, output=""}
    False {from=MaterialExpressionGetMaterialAttributes_1, output=EmissiveColor}
   ]
   properties {group=Flipbook 01 Global, parameterName=FB_ApplyEmissive, sortPriority=3}
  }
  MaterialExpressionStaticSwitchParameter_9 {
   class=MaterialExpressionStaticSwitchParameter
   inputs [
    True {from=MaterialExpressionAppendVector_14, output=""}
    False {from=MaterialExpressionScalarParameter_6, output=""}
   ]
   properties {group=Flipbook 04 Mouth, parameterName=FB_MouhtUVNonUniformScale, sortPriority=11}
  }
  MaterialExpressionStaticSwitch_0 {
   class=MaterialExpressionStaticSwitch
   inputs [
    True {from=MaterialExpressionNamedRerouteUsage_53, output=""}
    False {from=MaterialExpressionConstant_0, output=""}
    Value {from=MaterialExpressionNamedRerouteUsage_0, output=""}
   ]
   properties {}
  }
  MaterialExpressionStaticSwitch_1 {
   class=MaterialExpressionStaticSwitch
   inputs [
    True {from=MaterialExpressionNamedRerouteUsage_55, output=""}
    False {from=MaterialExpressionConstant_1, output=""}
    Value {from=MaterialExpressionNamedRerouteUsage_6, output=""}
   ]
   properties {}
  }
  MaterialExpressionStaticSwitch_10 {
   class=MaterialExpressionStaticSwitch
   inputs [
    True {from=MaterialExpressionMultiply_4, output=""}
    False {from=MaterialExpressionNamedRerouteUsage_94, output=""}
    Value {from=MaterialExpressionNamedRerouteUsage_95, output=""}
   ]
   properties {}
  }
  MaterialExpressionStaticSwitch_11 {
   class=MaterialExpressionStaticSwitch
   inputs [
    True {from=MaterialExpressionNamedRerouteUsage_91, output=""}
    False {from=MaterialExpressionNamedRerouteUsage_90, output=""}
    Value {from=MaterialExpressionStaticBoolParameter_14, output=""}
   ]
   properties {}
  }
  MaterialExpressionStaticSwitch_12 {
   class=MaterialExpressionStaticSwitch
   inputs [
    True {from=MaterialExpressionNamedRerouteDeclaration_68, output=""}
    False {from=MaterialExpressionConstant_12, output=""}
    Value {from=MaterialExpressionNamedRerouteUsage_92, output=""}
   ]
   properties {}
  }
  MaterialExpressionStaticSwitch_13 {
   class=MaterialExpressionStaticSwitch
   inputs [
    True {from=MaterialExpressionLinearInterpolate_2, output=""}
    False {from=MaterialExpressionReroute_1, output=""}
    Value {from=MaterialExpressionNamedRerouteUsage_97, output=""}
   ]
   properties {}
  }
  MaterialExpressionStaticSwitch_14 {
   class=MaterialExpressionStaticSwitch
   inputs [
    True {from=MaterialExpressionLinearInterpolate_1, output=""}
    False {from=MaterialExpressionReroute_5, output=""}
    Value {from=MaterialExpressionNamedRerouteUsage_102, output=""}
   ]
   properties {}
  }
  MaterialExpressionStaticSwitch_15 {
   class=MaterialExpressionStaticSwitch
   inputs [
    True {from=MaterialExpressionMaterialFunctionCall_7, output=Result}
    False {from=MaterialExpressionNamedRerouteUsage_105, output=""}
    Value {from=MaterialExpressionNamedRerouteUsage_4, output=""}
   ]
   properties {}
  }
  MaterialExpressionStaticSwitch_17 {
   class=MaterialExpressionStaticSwitch
   inputs [
    True {from=MaterialExpressionMaterialFunctionCall_9, output=Result}
    False {from=MaterialExpressionNamedRerouteUsage_103, output=""}
    Value {from=MaterialExpressionNamedRerouteUsage_5, output=""}
   ]
   properties {}
  }
  MaterialExpressionStaticSwitch_18 {
   class=MaterialExpressionStaticSwitch
   inputs [
    True {from=MaterialExpressionMaterialFunctionCall_8, output=Result}
    False {from=MaterialExpressionNamedRerouteUsage_104, output=""}
    Value {from=MaterialExpressionNamedRerouteUsage_11, output=""}
   ]
   properties {}
  }
  MaterialExpressionStaticSwitch_2 {
   class=MaterialExpressionStaticSwitch
   inputs [
    True {from=MaterialExpressionNamedRerouteUsage_56, output=""}
    False {from=MaterialExpressionConstant_2, output=""}
    Value {from=MaterialExpressionNamedRerouteUsage_12, output=""}
   ]
   properties {}
  }
  MaterialExpressionStaticSwitch_3 {
   class=MaterialExpressionStaticSwitch
   inputs [
    True {from=MaterialExpressionNamedRerouteUsage_60, output=""}
    False {from=MaterialExpressionConstant_3, output=""}
    Value {from=MaterialExpressionNamedRerouteUsage_15, output=""}
   ]
   properties {}
  }
  MaterialExpressionStaticSwitch_4 {
   class=MaterialExpressionStaticSwitch
   inputs [
    True {from=MaterialExpressionNamedRerouteUsage_61, output=""}
    False {from=MaterialExpressionConstant_4, output=""}
    Value {from=MaterialExpressionNamedRerouteUsage_16, output=""}
   ]
   properties {}
  }
  MaterialExpressionStaticSwitch_5 {
   class=MaterialExpressionStaticSwitch
   inputs [
    True {from=MaterialExpressionNamedRerouteUsage_62, output=""}
    False {from=MaterialExpressionConstant_5, output=""}
    Value {from=MaterialExpressionNamedRerouteUsage_26, output=""}
   ]
   properties {}
  }
  MaterialExpressionStaticSwitch_6 {
   class=MaterialExpressionStaticSwitch
   inputs [
    True {from=MaterialExpressionNamedRerouteUsage_66, output=""}
    False {from=MaterialExpressionConstant_6, output=""}
    Value {from=MaterialExpressionNamedRerouteUsage_29, output=""}
   ]
   properties {}
  }
  MaterialExpressionStaticSwitch_7 {
   class=MaterialExpressionStaticSwitch
   inputs [
    True {from=MaterialExpressionNamedRerouteUsage_67, output=""}
    False {from=MaterialExpressionConstant_7, output=""}
    Value {from=MaterialExpressionNamedRerouteUsage_54, output=""}
   ]
   properties {}
  }
  MaterialExpressionStaticSwitch_8 {
   class=MaterialExpressionStaticSwitch
   inputs [
    True {from=MaterialExpressionNamedRerouteUsage_68, output=""}
    False {from=MaterialExpressionConstant_8, output=""}
    Value {from=MaterialExpressionNamedRerouteUsage_57, output=""}
   ]
   properties {}
  }
  MaterialExpressionStaticSwitch_9 {
   class=MaterialExpressionStaticSwitch
   inputs [
    True {from=MaterialExpressionMultiply_2, output=""}
    False {from=MaterialExpressionNamedRerouteUsage_80, output=""}
    Value {from=MaterialExpressionNamedRerouteUsage_82, output=""}
   ]
   properties {}
  }
  MaterialExpressionStep_0 {
   class=MaterialExpressionStep
   inputs [
    Y {from="", output=""}
    X {from=MaterialExpressionAdd_6, output=""}
   ]
   properties {}
  }
  MaterialExpressionStep_1 {
   class=MaterialExpressionStep
   inputs [
    Y {from="", output=""}
    X {from=MaterialExpressionChannelMaskParameter_5, output=R}
   ]
   properties {constY=0.025}
  }
  MaterialExpressionStep_2 {
   class=MaterialExpressionStep
   inputs [
    Y {from=MaterialExpressionChannelMaskParameter_5, output=R}
    X {from="", output=""}
   ]
   properties {constX=0.975}
  }
  MaterialExpressionSubtract_1 {
   class=MaterialExpressionSubtract
   inputs [
    A {from=MaterialExpressionObjectPositionWS_0}
    B {from=MaterialExpressionWorldPosition_0, output=XYZ}
   ]
  }
  MaterialExpressionSubtract_11 {
   class=MaterialExpressionSubtract
   inputs [
    A {from=MaterialExpressionLinearInterpolate_7}
    B {from=MaterialExpressionAdd_3}
   ]
  }
  MaterialExpressionSubtract_2 {
   class=MaterialExpressionSubtract
   inputs [
    A {from=MaterialExpressionAdd_5}
    B {from=MaterialExpressionWorldPosition_2, output=XYZ}
   ]
  }
  MaterialExpressionSubtract_3 {
   class=MaterialExpressionSubtract
   inputs [
    A {from=MaterialExpressionComponentMask_1}
    B {from=MaterialExpressionMultiply_10}
   ]
  }
  MaterialExpressionSubtract_4 {
   class=MaterialExpressionSubtract
   inputs [
    A {from=MaterialExpressionMultiply_12}
    B {from=""}
   ]
   properties {constB=2}
  }
  MaterialExpressionSubtract_5 {
   class=MaterialExpressionSubtract
   inputs [
    A {from=MaterialExpressionActorPositionWS_0}
    B {from=MaterialExpressionWorldPosition_1, output=XYZ}
   ]
  }
  MaterialExpressionSubtract_6 {
   class=MaterialExpressionSubtract
   inputs [
    A {from=MaterialExpressionNamedRerouteUsage_84}
    B {from=""}
   ]
  }
  MaterialExpressionSubtract_7 {
   class=MaterialExpressionSubtract
   inputs [
    A {from=MaterialExpressionMultiply_13}
    B {from=MaterialExpressionNamedRerouteUsage_85}
   ]
  }
  MaterialExpressionSubtract_8 {
   class=MaterialExpressionSubtract
   inputs [
    A {from=MaterialExpressionMultiply_14}
    B {from=MaterialExpressionNamedRerouteUsage_86}
   ]
  }
  MaterialExpressionTextureCoordinate_0 {class=MaterialExpressionTextureCoordinate}
  MaterialExpressionTextureCoordinate_1 {class=MaterialExpressionTextureCoordinate, properties {coordinateIndex=1}}
  MaterialExpressionTextureCoordinate_2 {class=MaterialExpressionTextureCoordinate}
  MaterialExpressionTextureCoordinate_3 {class=MaterialExpressionTextureCoordinate, properties {coordinateIndex=1}}
  MaterialExpressionTextureCoordinate_4 {class=MaterialExpressionTextureCoordinate}
  MaterialExpressionTextureCoordinate_5 {class=MaterialExpressionTextureCoordinate, properties {coordinateIndex=1}}
  MaterialExpressionTextureCoordinate_7 {class=MaterialExpressionTextureCoordinate}
  MaterialExpressionTextureCoordinate_8 {class=MaterialExpressionTextureCoordinate, properties {coordinateIndex=1}}
  MaterialExpressionTextureObjectParameter_0 {
   class=MaterialExpressionTextureObjectParameter
   inputs [
    UVs {from="", output=""}
    Apply View MipBias {from="", output=""}
   ]
   properties {
    group=Flipbook 04 Mouth
    parameterName=FB_MouthFlipbookTexture
    sortPriority=3
    texture=/Game/Global/Materials/Character_Parent_2023/FlipbookFace_Parent_2023/Textures/T_black_emptyalpha
   }
  }
  MaterialExpressionTextureObjectParameter_1 {
   class=MaterialExpressionTextureObjectParameter
   inputs [
    UVs {from="", output=""}
    Apply View MipBias {from="", output=""}
   ]
   properties {
    group=Flipbook 03 Eyes
    parameterName=FB_EyeFlipbookTexture
    sortPriority=3
    texture=&T_black_emptyalpha
   }
  }
  MaterialExpressionTextureObjectParameter_2 {
   class=MaterialExpressionTextureObjectParameter
   inputs [
    UVs {from="", output=""}
    Apply View MipBias {from="", output=""}
   ]
   properties {
    group=Flipbook 02 Brows
    parameterName=FB_BrowTexture
    sortPriority=2
    texture=&T_black_emptyalpha
   }
  }
  MaterialExpressionVectorParameter_0 {
   class=MaterialExpressionVectorParameter
   properties {
    group=Flipbook 02 Brows
    parameterName=FB_BrowRemap_0-3
    sortPriority=201
    channelNames {a=3 - Brow Dn Glabella Up, b=2 - Brow Up Glabella Dn, g=1 - Brow Up Glabella Up, r=0 - Brow Neutral}
    defaultValue=<rgba 0 1 2 3>
   }
  }
  MaterialExpressionVectorParameter_1 {
   class=MaterialExpressionVectorParameter
   properties {
    group=Flipbook 02 Brows
    parameterName=FB_BrowRemap_4-7
    sortPriority=202
    channelNames {a=7 - Glabella Up, b=6 - Brow Down, g=5 - Brow Up, r=4 - Brow Dn Glabella Dn}
    defaultValue=<rgba 4 5 6 7>
   }
  }
  MaterialExpressionVectorParameter_10 {
   class=MaterialExpressionVectorParameter
   properties {group=Flipbook 06 Debug, parameterName=FB_DebugVertColorOffset, sortPriority=23, defaultValue=<rgba 0 0 0 1>}
  }
  MaterialExpressionVectorParameter_11 {
   class=MaterialExpressionVectorParameter
   properties {group=Flipbook 06 Debug, parameterName=FB_DebugUVOffset, sortPriority=2, defaultValue=<rgba 0 0 0 1>}
  }
  MaterialExpressionVectorParameter_12 {
   class=MaterialExpressionVectorParameter
   properties {group=Flipbook 01 Global, parameterName=FB_SRM, channelNames {a="", b=Metallic, g=Roughness, r=Specular}, defaultValue=<rgba 0.5 0.5 0 1>}
  }
  MaterialExpressionVectorParameter_13 {
   class=MaterialExpressionVectorParameter
   properties {group=Flipbook 01 Global, parameterName=FB_SubsurfaceColor, defaultValue=<rgba 0 0 0 1>}
  }
  MaterialExpressionVectorParameter_2 {
   class=MaterialExpressionVectorParameter
   properties {
    group=Flipbook 02 Brows
    parameterName=FB_BrowRemap_8-11
    sortPriority=203
    channelNames {a=Unused, b=Unused, g=Unused, r=8 - Glabella Dn}
    defaultValue=<rgba 8 9 10 11>
   }
  }
  MaterialExpressionVectorParameter_3 {
   class=MaterialExpressionVectorParameter
   properties {
    group=Flipbook 03 Eyes
    parameterName=FB_EyeRemap_0-3
    sortPriority=201
    channelNames {a=3 - Eye Blink, b=2 - Eye Squint, g=1 - Eye Squeeze, r=0 - Eye Neutral}
    defaultValue=<rgba 0 1 2 3>
   }
  }
  MaterialExpressionVectorParameter_4 {
   class=MaterialExpressionVectorParameter
   properties {
    group=Flipbook 03 Eyes
    parameterName=FB_EyeRemap_4-7
    sortPriority=202
    channelNames {a=7 - Eye Wide, b=6 - Brow Down, g=5 - Brow Dn Glabella Dn, r=4 - Brow Dn Glabella Up}
    defaultValue=<rgba 4 5 6 7>
   }
  }
  MaterialExpressionVectorParameter_5 {
   class=MaterialExpressionVectorParameter
   properties {group=Flipbook 02 Brows, parameterName=FB_BrowColor, sortPriority=104, defaultValue=<rgba 0 0 0 1>}
  }
  MaterialExpressionVectorParameter_6 {
   class=MaterialExpressionVectorParameter
   properties {
    group=Flipbook 04 Mouth
    parameterName=FB_MouthRemap_0-3
    sortPriority=201
    channelNames {a=3 - Jaw Open R Smile, b=2 - Jaw Open L Smile, g=1 - Mouth Pucker, r=0 - Mouth Neutral}
    defaultValue=<rgba 0 1 2 3>
   }
  }
  MaterialExpressionVectorParameter_7 {
   class=MaterialExpressionVectorParameter
   properties {
    group=Flipbook 04 Mouth
    parameterName=FB_MouthRemap_4-7
    sortPriority=202
    channelNames {a=7 - Jaw Open, b=6 - Jaw Open R Snear, g=5 -Jaw Open L Snear, r=4 - Jaw Open Frown}
    defaultValue=<rgba 4 5 6 7>
   }
  }
  MaterialExpressionVectorParameter_8 {
   class=MaterialExpressionVectorParameter
   properties {
    group=Flipbook 04 Mouth
    parameterName=FB_MouthRemap_8-11
    sortPriority=203
    channelNames {a=11 - Smile, b=10 - Frown Wide, g=9 - Smile R Wide, r=8 - Smile L Wide}
    defaultValue=<rgba 4 5 6 7>
   }
  }
  MaterialExpressionVectorParameter_9 {
   class=MaterialExpressionVectorParameter
   properties {
    group=Flipbook 04 Mouth
    parameterName=FB_MouthRemap_12-15
    sortPriority=204
    channelNames {a=Unused, b=14 - Snear R, g=13 - Snear L, r=12 - Frown}
    defaultValue=<rgba 4 5 6 7>
   }
  }
  MaterialExpressionVertexColor_0 {class=MaterialExpressionVertexColor, inputs [], properties {}}
  MaterialExpressionVertexColor_3 {class=MaterialExpressionVertexColor, inputs [], properties {}}
  MaterialExpressionVertexInterpolator_0 {class=MaterialExpressionVertexInterpolator, inputs [VS {from=MaterialExpressionLocalPosition_0, output=XYZ}], properties {}}
  MaterialExpressionWorldPosition_0 {class=MaterialExpressionWorldPosition, inputs [], properties {}}
  MaterialExpressionWorldPosition_1 {class=MaterialExpressionWorldPosition, inputs [], properties {}}
  MaterialExpressionWorldPosition_2 {class=MaterialExpressionWorldPosition, inputs [], properties {}}
 }}
