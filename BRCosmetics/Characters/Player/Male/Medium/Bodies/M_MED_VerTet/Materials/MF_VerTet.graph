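# Material function for the VerTet male character body that adds an animated, tinted tech-glow pulse to incoming material attributes.
# Character gradients (radial by default) and triplanar noise on pre-skinned positions feed a time-driven band, declared as GradientPulse alongside FX-derived circuit and Fresnel masks.
# TechGlowMask and existing emissive shape the glow, added to BaseColor and optionally darkened Emissive.
# The glow mask also drives temporal responsiveness.
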MaterialFunction /BRCosmetics/Characters/Player/Male/Medium/Bodies/M_MED_VerTet/Materials/MF_VerTet
graph {nodes [
  n1 = GetMaterialAttributes(None=FunctionInput() {inputName=MA_In, inputType=FunctionInput_MaterialAttributes}) {attributeGetTypes [#1, #2]}
  n2 = ComponentMask(None=@n1.EmissiveColor) {g=true}
  n3 = Saturate(None=Add(A=@n1.EmissiveColor, B=Multiply(A=@n2) {constB=2}))
  n4 = TextureSampleParameter2D() {
   parameterName=TechGlowMask
   samplerSource=SSM_Wrap_WorldGroupSettings
   samplerType=SAMPLERTYPE_Masks
   texture=/BRCosmetics/Characters/Player/Male/Medium/Bodies/M_MED_VerTet/Textures/T_VerTet_Body_FX2
  }
  n5 = Multiply(A=@n4.R, B=NamedRerouteUsage())
  n6 = Multiply(
   A=Multiply(A=LinearInterpolate(Alpha=@n3) {constA=0.05, constB=2}, B=@n5)
   B=VectorParameter() {parameterName=EmissiveGlowColor, defaultValue=<rgba 2 0.1267 1.4425 1>}.RGB
  )
  n7 = LinearInterpolate(B=ScalarParameter() {defaultValue=0.25, group=EmissivePulse, parameterName=EmissiveDarken}, Alpha=NamedRerouteUsage())
  n8 = ScalarParameter() {controlType=Enumeration, defaultValue=1, enumeration=/Game/Global/Materials/Enums/E_Mat_OnOff, parameterName=UseEmissiveDarken}
  n9 = LinearInterpolate(A=LinearInterpolate(A=@n1.EmissiveColor, B=Multiply(A=@n1.EmissiveColor, B=@n7) {constB=0.25}, Alpha=@n8), B=@n1.EmissiveColor, Alpha=@n5)
  FunctionOutput(None=SetMaterialAttributes!(MaterialAttributes=@n1.MaterialAttributes, BaseColor=Add(A=@n1.BaseColor, B=@n6), EmissiveColor=Add(A=@n9, B=@n6)) {attributeSetTypes [#1, #2]})
  n10 = PreSkinnedPosition()
  n11 = ScalarParameter() {defaultValue=0.4139, group=EmissivePulse, parameterName=CharacterGradient_RumpFalloffHardness, sliderMax=1, sortPriority=1000}
  n12 = SphereMask(
   A=ComponentMask(None=VertexInterpolator(VS=@n10).PS) {g=true, r=true}
   B=Constant2Vector()
   Radius=ScalarParameter() {defaultValue=25, group=EmissivePulse, parameterName=CharacterGradient_RumpRadius, sortPriority=1000}
   Hardness=@n11
  ) {attenuationRadius=30, hardnessPercent=80}
  n13 = TextureSample() {
   samplerSource=SSM_Clamp_WorldGroupSettings
   samplerType=SAMPLERTYPE_Alpha
   texture=/BRCosmetics/Characters/Player/Male/Medium/Bodies/M_MED_VerTet/Textures/T_VerTet_Body_ArmsMask
  }
  n14 = ScalarParameter() {defaultValue=2, group=EmissivePulse, parameterName=Radial_HorizontalCompression, sliderMax=2, sliderMin=0.01, sortPriority=16}
  n15 = MaterialFunctionCall!(
   Local Position
   ArmBalance=ScalarParameter() {defaultValue=0.7532, group=EmissivePulse, parameterName=Arm Balance, sliderMax=1.5, sortPriority=-10}
   RumpBalance=ScalarParameter() {defaultValue=1.14, group=EmissivePulse, parameterName=Rump Balance, sliderMax=1.5, sortPriority=-10}
   Arms Mask=StaticSwitchParameter(True=OneMinus(None=@n12), False=@n13.R) {group=EmissivePulse, parameterName=Use Cylinder Mask, sortPriority=320}
   Arm_GradientEndPosition
   Arm_GradientBeginPosition=VectorParameter() {parameterName=RumpGradientStart, defaultValue=<rgba 3 3 0 0>}.RGB
   Arm_GradientDirection
   Rump_GradientEndHeight=ScalarParameter() {group=EmissivePulse, parameterName=RumpGradient_FeetHeight, sliderMax=180, sortPriority=64}
   Rump_GradientBeginHeight=ScalarParameter() {defaultValue=169.44, group=EmissivePulse, parameterName=RumpGradient_TopHeight, sliderMax=180, sortPriority=64}
   Radial_CenterHeight=ScalarParameter() {defaultValue=0.76, group=EmissivePulse, parameterName=Radial_CenterHeightPercent, sliderMax=1, sortPriority=16}
   Radial_HorizontalCompression=@n14
   Disable Arm Gradients=StaticBoolParameter() {group=EmissivePulse, parameterName=Disable Arm Gradient}
   Disable Rump Gradient=StaticBoolParameter() {group=EmissivePulse, parameterName=Disable Rump Gradient}
  ) {materialFunction=/Game/Global/MaterialFunctions/TechArt/MF_CharacterGradients}
  n16 = StaticSwitchParameter(True=@n15.RadialGradient, False=@n15.RadialGradient) {defaultValue=true, group=1_Features, parameterName="Radial (true) or Linear (false)"}
  n17 = ScalarParameter() {defaultValue=24, group=EmissivePulse, parameterName=TriplanarNoiseSize}
  n18 = MaterialFunctionCall(
   TextureObject=TextureObjectParameter() {parameterName=PulseNoise, texture=/Game/Global/Textures/Default/Noise/T_SoftNoise}
   TextureSize=MaterialFunctionCall(R=@n17, G=@n17, B=@n17) {materialFunction=&AppendMany}.RGB
   WorldPosition=VertexInterpolator(VS=PreSkinnedPosition()).PS
  ) {materialFunction=/Game/Weapons/WeaponSkins/Wraps/MaterialFunctions/WorldAlignedTexture}
  n19 = Add(
   A=LinearInterpolate(A=@n16, B=OneMinus(None=@n16), Alpha=ScalarParameter() {group=EmissivePulse, parameterName=InvertRump})
   B=Multiply(A=ConstantBiasScale(None=@n18.XY Texture) {bias=-0.5, scale=2}) {constB=0.05}
  )
  Multiply(A=@n19) {constB=0.8}
  n20 = Time()
  n21 = ScalarParameter() {parameterName=TimePulseOffset}
  n22 = Frac(None=Multiply(A=Add(A=@n20, B=@n21), B=ScalarParameter() {defaultValue=0.5, group=EmissivePulse, parameterName=PulseSpeed}) {constB=0.4})
  n23 = Subtract(A=Subtract(A=Multiply(A=@n22) {constB=1.25}) {constB=0.25}, B=@n19)
  n24 = Saturate(None=Multiply(A=@n23, B=ScalarParameter() {defaultValue=1, parameterName=MultEnd}))
  n25 = Add(
   A=Multiply(A=@n23, B=ScalarParameter() {defaultValue=40, parameterName=MultStart}) {constB=40}
   B=ScalarParameter() {defaultValue=6, parameterName=AddStart}
  ) {constB=5}
  n26 = Multiply(
   A=Power(Base=OneMinus(None=@n24), Exp=ScalarParameter() {defaultValue=12, parameterName=PowEnd}) {constExponent=6}
   B=Power(Base=Saturate(None=@n25), Exp=ScalarParameter() {defaultValue=6, parameterName=PowStart}) {constExponent=6}
  )
  n27 = ComponentMask(None=@n26) {r=true}
  n28 = ComponentMask(None=@n26) {g=true}
  n29 = Add(A=Add(A=Multiply(A=@n27) {constB=0.5}, B=Multiply(A=@n28) {constB=0.5}), B=Multiply(A=ComponentMask(None=@n26) {b=true}) {constB=0.5})
  NamedRerouteDeclaration(None=Multiply(A=Saturate(None=@n29), B=ScalarParameter() {defaultValue=0.6, parameterName=PulseStrength})) {name=GradientPulse, nodeColor=<rgba 1 0.8438 0 1>}
  n30 = ComponentMask(None=FunctionInput() {inputName=FX}) {r=true}
  n31 = Fresnel(ExponentIn=Constant() {r=0.8}) {exponent=2}
  n32 = Multiply(A=Power(Base=OneMinus(None=@n31), Exp=Constant() {r=3.7}) {constExponent=5}) {constB=2}
  NamedRerouteDeclaration(None=Multiply(A=@n30, B=Saturate(None=@n32))) {name=Fresnel, nodeColor=<rgba 0.125 0 1 1>}
  NamedRerouteDeclaration(None=@n30) {name=CircuitsMask, nodeColor=<rgba 1 0 0.6094 1>}
  TemporalResponsivenessOutput!(None=LinearInterpolate(Alpha=@n4.R) {constB=0.4})
 ]}
