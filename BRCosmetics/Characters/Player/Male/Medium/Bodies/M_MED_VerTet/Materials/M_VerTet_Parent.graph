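# Parent skeletal mesh material for the VerTet male player body cosmetic, built on the shared 2023 character parent functions.
# PreFX output passes through Ice and Galaxy crew-style layers, optional carbon fiber, glitter-holographic and VerTet-specific functions, then PostFX.
# The isGold static switch enables the dual-layer carbon fiber and blends its wrap normal into the clear coat normal by mask.
# Named reroutes share PreFX masks and fresnel across layers.
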Material /BRCosmetics/Characters/Player/Male/Medium/Bodies/M_MED_VerTet/Materials/M_VerTet_Parent
graph {
 properties {
  bAutomaticallySetUsageInEditor=false
  bUseMaterialAttributes=true
  bUsedWithMeshDeformer=true
  bUsedWithMorphTargets=true
  bUsedWithSkeletalMesh=true
  pixelDepthOffsetMode=PDOM_Legacy
  previewMesh=/BRCosmetics/Characters/Player/Male/Medium/Bodies/M_MED_VerTet/Meshes/M_MED_VerTet
  shadingModel=MSM_FromMaterialExpression
  usedShadingModels=Default Lit | Subsurface | Clear Coat | Subsurface Profile | Cloth
 }
 nodes [
  n1 = MaterialFunctionCall() {materialFunction=&MF_CharacterParent_2023_PreFX}
  n2 = MaterialFunctionCall!(
   Tech Art Mask=NamedRerouteUsage()
   Mask=NamedRerouteUsage()
   In=MaterialFunctionCall!(TechArt=NamedRerouteUsage(), Fresnel=NamedRerouteUsage(), In=@n1.Result) {materialFunction=/Game/Global/Materials/Character_Parent_2023/CrewStyle_Parent_2023/Ice/MF_CrewStyle_Ice_Inst}.Result
  ) {materialFunction=/Game/Global/Materials/Character_Parent_2023/CrewStyle_Parent_2023/Galaxy/MF_CrewStyle_Galaxy_Inst}
  n3 = MaterialFunctionCall!(Alpha=NamedRerouteUsage(), In=@n2.Result) {materialFunction=/Game/Global/MaterialFunctions/TechArt/MF_CarbonFiberDualLayer}
  n4 = GetMaterialAttributes(None=@n3.Result) {attributeGetTypes [#1]}
  n5 = MaterialFunctionCall(
   BaseNormal=@n4.Normal
   AdditionalNormal=TextureSample(UVs=@n3.CarbonFiberUvs) {samplerType=SAMPLERTYPE_Normal, texture=/Game/Weapons/WeaponSkins/Wraps/Textures/T_Wrap_CarbonFiber_N}.RGB
  ) {materialFunction=&BlendAngleCorrectedNormals}
  ClearCoatNormalCustomOutput!(None=StaticSwitchParameter(True=LinearInterpolate(A=@n4.Normal, B=@n5.Result, Alpha=@n3.Mask), False=@n4.Normal) {parameterName=isGold})
  n6 = NamedRerouteUsage()
  n7 = StaticSwitchParameter(True=@n3.Result, False=@n2.Result) {parameterName=isGold}
  n8 = NamedRerouteUsage()
  MaterialFunctionCall(
   Material=MaterialFunctionCall!(
    FX=NamedRerouteUsage()
    MA_In=MaterialFunctionCall!(Alpha=@n6, In=@n7, Fresnel=@n8) {materialFunction=/Game/Global/MaterialFunctions/TechArt/MF_GlitterHolographic}.Result
   ) {materialFunction=&/MF_VerTet}.Result
   M=@n1.M
   Pre Skinned Local Position=@n1.M
   Distance Blend=@n1.M
   Fresnel=@n1.M
  ) {materialFunction=&MF_CharacterParent_2023_PostFX}
  NamedRerouteDeclaration(None=@n1.Fresnel) {name=fresnel, nodeColor=<rgba 1 0.6094 0 1>}
  NamedRerouteDeclaration(None=@n1.Tech Art Mask) {name=TA Mask, nodeColor=<rgba 0.2891 0 1 1>}
  NamedRerouteDeclaration(None=@n1.M) {name=Mask, nodeColor=<rgba 0.8984 0 1 1>}
 ]
}
