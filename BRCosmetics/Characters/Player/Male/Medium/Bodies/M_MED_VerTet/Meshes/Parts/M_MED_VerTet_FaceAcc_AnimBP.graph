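# Animation Blueprint for the M_MED_VerTet face accessory mesh that tells its materials when the facial animation changes.
# Each update, flipbook faces compare brow, eye and mouth index curves with cached values to set HasPoseChanged, then refresh the cache.
# With material-driven pupils, look and blink curves set HasLookDirectionChanged and HasLidChanged using tolerance comparisons.
# The previous look direction is cached, but the previous blink pose is never updated in this graph.
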AnimBlueprint /BRCosmetics/Characters/Player/Male/Medium/Bodies/M_MED_VerTet/Meshes/Parts/M_MED_VerTet_FaceAcc_AnimBP
target_skeleton=&/M_MED_VerTet_FaceAcc_Skeleton
graphs [EventGraph {nodes [
   n1 = Event BlueprintUpdateAnimation()
   > Parent: BlueprintUpdateAnimation!(execute, DeltaTimeX:float=@n1.DeltaTimeX) -> (then) {is=title K2Node_CallParentFunction}
   > n2 = Sequence()
   > Branch(Condition=Get Is Flipbook Face() -> (Is Flipbook Face:bool))
   n3 = GetCurveValue(CurveName=flipbook_face_L_brow_index)
   n4 = "Not Equal (Float)"(A=@n3, B=Get L_Brow_Prev_Index_0() -> (L_Brow_Prev_Index_0:float))
   n5 = GetCurveValue(CurveName=flipbook_face_R_brow_index)
   n6 = "Not Equal (Float)"(A=@n5, B=Get R_Brow_Prev_Index_0() -> (R_Brow_Prev_Index_0:float))
   n7 = GetCurveValue(CurveName=flipbook_face_L_eye_index)
   n8 = "Not Equal (Float)"(A=@n7, B=Get L_ Eye_Prev_Index_0() -> (L_ Eye_Prev_Index_0:float))
   n9 = GetCurveValue(CurveName=flipbook_face_R_eye_index)
   n10 = "Not Equal (Float)"(A=@n9, B=Get R_Eye_Prev_Index_0() -> (R_Eye_Prev_Index_0:float))
   n11 = GetCurveValue(CurveName=flipbook_face_mouth_index)
   n12 = "Not Equal (Float)"(A=@n11, B=Get Mouth_Prev_Index_0() -> (Mouth_Prev_Index_0:float))
   > SetScalarParameterValueOnMaterials!(
    execute
    self:Mesh Component Object Reference=GetPartSkeletalMeshComponent()
    ParameterName:Name=HasPoseChanged
    ParameterValue:float="To Float (Boolean)"(InBool=OR Boolean!(A=@n8, B=@n10, C:bool=@n12, D:bool=@n4, E:bool=@n6) -> (ReturnValue))
   ) -> (then)
   > Set L_ Eye_Prev_Index_0(L_ Eye_Prev_Index_0:float=@n7) -> (Output_Get:float)
   > Set R_Eye_Prev_Index_0(R_Eye_Prev_Index_0:float=@n9) -> (Output_Get:float)
   > Set Mouth_Prev_Index_0(Mouth_Prev_Index_0:float=@n11) -> (Output_Get:float)
   > Set L_Brow_Prev_Index_0(L_Brow_Prev_Index_0:float=@n3) -> (Output_Get:float)
   > Set R_Brow_Prev_Index_0(R_Brow_Prev_Index_0:float=@n5) -> (Output_Get:float)
   Branch(execute=@n2.then_1, Condition=Get bHasMaterialDrivenPupils_0() -> (bHasMaterialDrivenPupils_0:bool))
   n13 = MakeVector2D(
    X=float - float(A=GetCurveValue(CurveName=look_left_pose), B=GetCurveValue(CurveName=look_right_pose))
    Y=float - float(A=GetCurveValue(CurveName=look_up_pose), B=GetCurveValue(CurveName=look_down_pose))
   )
   n14 = Get Prev_LookDirection_0() -> (Prev_LookDirection_0:Vector 2D Structure)
   > SetScalarParameterValueOnMaterials(
    self=GetPartSkeletalMeshComponent()
    ParameterName=HasLookDirectionChanged
    ParameterValue="To Float (Boolean)"(InBool="Not Equal (Vector2D)"!(A:Vector 2D Structure=@n13, B:Vector 2D Structure=@n14, ErrorTolerance:float=0.025) -> (ReturnValue:bool=~) {is=title K2Node_PromotableOperator})
   )
   > Set Prev_LookDirection_0(Prev_LookDirection_0:Vector 2D Structure=@n13) -> (Output_Get:Vector 2D Structure)
   n15 = MakeVector2D(X=GetCurveValue(CurveName=L_blink_pose), Y=GetCurveValue(CurveName=R_blink_pose))
   n16 = Get Prev Blink Pose_0() -> (Prev Blink Pose_0:Vector 2D Structure)
   n17 = Get Blink Pose Threshold_0() -> (Blink Pose Threshold_0:float)
   > SetScalarParameterValueOnMaterials(
    self=GetPartSkeletalMeshComponent()
    ParameterName=HasLidChanged
    ParameterValue="To Float (Boolean)"(InBool="Not Equal (Vector2D)"(A=@n15, B=@n16, ErrorTolerance=@n17) -> (ReturnValue=~))
   )
  ]}]
variables [
 bHasMaterialDrivenPupils_0 {type=boolean}
 bIsFlipbook_face_0 {type=boolean, default=true}
 Blink Pose Threshold_0 {type=number}
 Is Flipbook Face {type=boolean, default=true}
 L_ Eye_Prev_Index_0 {type=number}
 L_Brow_Prev_Index_0 {type=number}
 Mouth_Prev_Index_0 {type=number}
 Prev Blink Pose_0 {type=Vector2D, default=<0 0>}
 Prev_LookDirection_0 {type=Vector2D, default=<0 0>}
 R_Brow_Prev_Index_0 {type=number}
 R_Eye_Prev_Index_0 {type=number}
]
